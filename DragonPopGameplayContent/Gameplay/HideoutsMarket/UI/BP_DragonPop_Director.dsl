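; BP_DragonPop_Director is a Blueprint in the DragonPop HideoutsMarket UI content that exposes a type check for buyable items.
; IsValidType takes a BGA Buyable Item and always returns true, so no item is filtered out.
; BeginPlay and Tick hold no logic, and the construction script only calls its parent.
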(event EventBeginPlay
  (|CallParentFunction) ; This node is disabled and will not be called.\nDrag off pins to build functionality.
)

(event EventTick
  (params (DeltaSeconds :type Float%20%28single-precision%29)))

(fn IsValidType
  (params (BuyableItem :type BGA%20Buyable%20Item%20Object%20Reference))
  (return true))

(fn ConstructionScript
  (|CallParentFunction))
