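; Dynamic UI director that shows and hides an item details scene for buyable Hideouts Market items.
; It binds to HUD interaction events, checking the receiving buyable item's display definition and IsValidType, which returns false here, presumably for child overrides.
; Valid targets activate the scene through a delta gate; otherwise a short timer delays removal for the outro.
; EndPlay unbinds OnPlayerCanInteract but binds OnInteractUpdated again rather than unbinding it.
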(event EventBeginPlay
  (bind _outputdelegate (EventDispatchers|CreateEvent :!event OnPlayerCanInteract))
  (bind _returnvalue (Class|FortHUDContext|GetFortHUDContextbyPlayer (DynamicUIDirector|GetOwningLocalPlayer)))
  (HUD|BindEventtoOnInteractUpdated _returnvalue _outputdelegate)
  (HUD|BindEventtoOnPlayerCanInteract _returnvalue _outputdelegate))

(event EventEndPlay
  (params (EndPlayReason :type EEndPlayReason%20Enum :default "Destroyed"))
  (bind _outputdelegate (EventDispatchers|CreateEvent :!event OnPlayerCanInteract))
  (bind _returnvalue (Class|FortHUDContext|GetFortHUDContextbyPlayer (DynamicUIDirector|GetOwningLocalPlayer)))
  (Utilities|IsValid _returnvalue
    (:Is%20Valid
      (HUD|BindEventtoOnInteractUpdated _returnvalue _outputdelegate)
      (HUD|UnbindEventfromOnPlayerCanInteract _returnvalue _outputdelegate))))

(fn IsValidType
  (params (BuyableItem :type BGA%20Buyable%20Item%20Object%20Reference))
  (return false))

(fn OnItemDetailsDeactivateTimerFinished
  (CallFunction|SetItemDetailsSceneActive false))

(fn OnPlayerCanInteract
  (params (ContextInfo :type Fort%20Interact%20Context%20Info%20Object%20Reference))
  (Utilities|IsValid ContextInfo
    (:Is%20Valid
      (bind _asbga_buyable_item (Utilities|Casting|CastToBGA_Buyable_Item (Variables|/Script/FortniteGame.FortInteractContextInfo|GetReceivingActor ContextInfo))
        (:then
          (Utilities|IsValid (Variables|/DragonPopGameplayContent/Gameplay/HideoutsMarket/BGA_Buyable_Item.BGA_Buyable_Item_C|GetDisplayItemDef _asbga_buyable_item)
            (:Is%20Valid
              (if (CallFunction|IsValidType _asbga_buyable_item)
                (CallFunction|SetItemDetailsSceneActive true)
                (Utilities|Time|ClearandInvalidateTimerbyHandle (Variables|Default|GetDeactivateItemDetailsTimer))
                (else
                  _)))
            (:Is%20Not%20Valid
              _))
          (if (and (Variables|Default|GetIsDetailsSceneActive) (not (Utilities|Time|IsTimerActivebyHandle (Variables|Default|GetDeactivateItemDetailsTimer))))
            (bind _returnvalue (Utilities|Time|SetTimerbyEvent (EventDispatchers|CreateEvent :!event OnItemDetailsDeactivateTimerFinished) 1.0)) ; Give outro time to play
            (Variables|Default|SetDeactivateItemDetailsTimer _returnvalue)))
        (:CastFailed
          _)))
    (:Is%20Not%20Valid
      _)))

(fn SetItemDetailsSceneActive
  (params (IsActive :type Boolean))
  (bind _itemdetailsscene (Variables|Default|GetItemDetailsScene))
  (bind _updated_variable (Utilities|FlowControl|DeltaGateWildcard (Variables|Default|GetIsDetailsSceneActive) IsActive)
    (:Success
      (if _updated_variable
        (DynamicUIDirector|AddScene _itemdetailsscene)
        (else
          (DynamicUIDirector|RemoveScene _itemdetailsscene))))))

(fn ConstructionScript)
