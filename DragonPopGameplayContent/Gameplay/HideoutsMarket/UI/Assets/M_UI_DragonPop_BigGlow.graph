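# Additive UI material for a big, soft radial glow, part of the DragonPop Hideouts Market UI assets.
# An exponential radial gradient blends between two orange color parameters, Big Glow Color 01 and 02.
# The same gradient is also multiplied by itself, giving a squared, tighter falloff.
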Material /DragonPopGameplayContent/Gameplay/HideoutsMarket/UI/Assets/M_UI_DragonPop_BigGlow
graph {
 properties {bAutomaticallySetUsageInEditor=false, blendMode=BLEND_Additive, materialDomain=MD_UI, usedShadingModels=Default Lit}
 nodes [
  n1 = MaterialFunctionCall(Radius=Constant() {r=0.5}, Density=Constant() {r=1}) {materialFunction=&RadialGradientExponential}
  LinearInterpolate(
   A=VectorParameter() {parameterName=Big Glow Color 01, defaultValue=<rgba 1 0.2882 0 1>}.RGB
   B=VectorParameter() {parameterName=Big Glow Color 02, defaultValue=<rgba 1 0.4222 0 1>}.RGB
   Alpha=@n1.RadialGradientExponential
  )
  Multiply(A=@n1.RadialGradientExponential, B=@n1.RadialGradientExponential)
 ]
}
