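# Translucent UI material for a DragonPop blurb shape in the Hideouts Market UI.
# A smoothstep over the red channel of a grayscale blurb texture sharpens it into a soft-edged mask.
# A vertical linear gradient blends two color constants, both currently black, so the resulting fill is flat.
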Material /DragonPopGameplayContent/Gameplay/HideoutsMarket/UI/Assets/M_UI_DragonPop_Blurb
graph {
 properties {bAutomaticallySetUsageInEditor=false, blendMode=BLEND_Translucent, materialDomain=MD_UI, usedShadingModels=Default Lit}
 nodes [
  Constant()
  LinearInterpolate(
   A=Constant3Vector() {constant=<rgba 0 0 0 1>}
   B=Constant3Vector() {constant=<rgba 0 0 0 1>}
   Alpha=MaterialFunctionCall() {materialFunction=&LinearGradient}.VGradient
  )
  SmoothStep(Value=TextureSample() {samplerType=SAMPLERTYPE_LinearGrayscale, texture=&/T_UI_DragonPop_Blurb}.R) {constMax=0.5, constMin=0.4}
 ]
}
