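# Translucent UI material for a DragonPop item background in the Hideouts Market, styled as a flame shape with outline and embers.
# Channels of one ItemBg texture yield anti-aliased main, inner and offset shadow shapes, plus an outline with sine shake and noise UV distortion.
# Named reroutes carry these masks, reveal-controlled color and ember masks, and panning ambient noise.
# Layered lerps apply flame, outline and inner colors, modulated by a graffiti texture.
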Material /DragonPopGameplayContent/Gameplay/HideoutsMarket/UI/Assets/M_UI_DragonPop_ItemBg
graph {
 properties {bAutomaticallySetUsageInEditor=false, blendMode=BLEND_Translucent, materialDomain=MD_UI, usedShadingModels=Default Lit}
 nodes [
  Constant2Vector() {g=-0.02, r=-0.048}
  Constant()
  n1 = Constant3Vector()
  n2 = LinearInterpolate(
   A=VectorParameter() {parameterName=Flame Color 01, defaultValue=<rgba 1 0.3961 0.0014 1>}.RGB
   B=VectorParameter() {parameterName=Flame Color 02, defaultValue=<rgba 0.5703 0.0673 0 1>}.RGB
   Alpha=NamedRerouteUsage()
  )
  n3 = Multiply(A=NamedRerouteUsage(), B=NamedRerouteUsage())
  n4 = LinearInterpolate(
   A=LinearInterpolate(
    A=LinearInterpolate(A=@n1, B=@n2, Alpha=@n3)
    B=VectorParameter() {parameterName=Outline Color, defaultValue=<rgba 0.8958 0.4116 0 1>}.RGB
    Alpha=NamedRerouteUsage()
   )
   B=VectorParameter() {parameterName=Inner Flame Color, defaultValue=<rgba 1 0.5196 0.0002 1>}.RGB
   Alpha=Multiply(A=NamedRerouteUsage()) {constB=0.3}
  )
  LinearInterpolate(
   A=@n4
   B=Multiply(A=@n4, B=Add(A=@n4) {constB=0.15})
   Alpha=TextureSample() {samplerType=SAMPLERTYPE_LinearColor, texture=&/T_UI_DragonPop_Graffiti}.RGB
  )
  MaterialFunctionCall() {materialFunction=&DebugTimeSine}
  MaterialFunctionCall(Speed=Constant() {r=0.1}) {materialFunction=&DebugTimeSine}
  n5 = NamedRerouteUsage()
  n6 = NamedRerouteUsage()
  n7 = Max(A=Max(A=Max(A=@n5, B=@n6), B=NamedRerouteUsage()), B=NamedRerouteUsage())
  Max(A=@n7, B=NamedRerouteUsage())
  Multiply(A=@n7, B=NamedRerouteUsage())
  n8 = TextureCoordinate()
  n9 = Constant2Vector() {g=-0.048, r=-0.024}
  n10 = TextureSample(UVs=MaterialFunctionCall(UVs=Add(A=@n8, B=@n9), Texture Scale=Constant() {r=0.4}) {materialFunction=&ScaleUVsByCenter}.UVs) {texture=&/T_UI_DragonPop_ItemBg}
  n11 = Constant() {r=0.1}
  n12 = Constant()
  NamedRerouteDeclaration(None=Multiply(
    A=MaterialFunctionCall(DistanceField=@n10.R, Threshold=@n11, Smoothing=@n12) {materialFunction=&MF_UI_AntiAliasing}.Outer
    B=ScalarParameter() {defaultValue=1, parameterName=InnerAlphaOpacity}
   )) {name=InnerAlpha, nodeColor=<rgba 0 0 0 1>}
  n13 = NamedRerouteUsage()
  n14 = TextureSample(UVs=NamedRerouteUsage()) {samplerType=SAMPLERTYPE_LinearColor, texture=/Game/Weapons/WeaponSkins/Textures/T_Swirly_Noise_highContrast1}
  n15 = ScalarParameter() {parameterName=OutlineDistortion}
  NamedRerouteDeclaration(None=LinearInterpolate(
    A=ComponentMask(None=@n13) {g=true, r=true}
    B=ComponentMask(None=@n14.RGB) {g=true, r=true}
    Alpha=LinearInterpolate(Alpha=@n15) {constB=0.1}
   )) {name=UV Distortion, nodeColor=<rgba 1 0.1641 0 1>}
  n16 = MaterialFunctionCall(Speed=ScalarParameter() {defaultValue=8, parameterName=OutlineShakeSpeed}) {materialFunction=&DebugTimeSine}
  n17 = ScalarParameter() {parameterName=Outline Movement}
  n18 = Add(
   A=NamedRerouteUsage()
   B=LinearInterpolate(
    A=Constant2Vector() {g=0.031, r=0.04}
    B=Constant2Vector() {g=-0.0008, r=-0.002}
    Alpha=LinearInterpolate(B=@n16.Result, Alpha=@n17) {constB=0}
   ) {constAlpha=1}
  )
  n19 = ScalarParameter() {defaultValue=0.95, parameterName=OutlineScale}
  n20 = SmoothStep(Value=TextureSample(UVs=MaterialFunctionCall(UVs=@n18, Texture Scale=@n19) {materialFunction=&ScaleUVsByCenter}.UVs) {texture=&T_UI_DragonPop_ItemBg}.A) {constMax=0.6, constMin=0.55}
  n21 = ScalarParameter() {defaultValue=0.9, parameterName=Outline Opacity}
  NamedRerouteDeclaration(None=NamedRerouteDeclaration(None=Multiply(A=@n20, B=@n21)) {name=Outline, nodeColor=<rgba 0 0 0 1>}) {name=OutlineWithGlow, nodeColor=<rgba 0.2188 0 1 1>}
  n22 = TextureSample() {samplerType=SAMPLERTYPE_LinearGrayscale, texture=&/T_UI_DragonPop_RevealMask}
  n23 = Step(Y=Add(A=ScalarParameter() {defaultValue=1, parameterName=ColorReveal, sliderMax=1}, B=@n22.R))
  NamedRerouteDeclaration(None=Multiply(A=OneMinus(None=@n23))) {name=ColorReveal, nodeColor=<rgba 0 0 0 1>}
  NamedRerouteDeclaration(None=TextureCoordinate()) {name=TexCoord, nodeColor=<rgba 0 0 0 1>}
  n24 = Constant() {r=0.296}
  n25 = LinearInterpolate(Alpha=ScalarParameter() {parameterName=Ember Y}) {constA=1.258, constB=1.474}
  n26 = Add(A=TextureCoordinate(), B=Multiply(A=AppendVector(A=@n24, B=@n25)) {constB=0.3})
  n27 = ScalarParameter() {defaultValue=0.296, parameterName=Ember Scale}
  n28 = SmoothStep(Value=TextureSample(UVs=MaterialFunctionCall(UVs=@n26, Texture Scale=@n27) {materialFunction=&ScaleUVsByCenter}.UVs) {samplerType=SAMPLERTYPE_LinearGrayscale, texture=&/T_UI_DragonPop_Ember}.B) {constMax=0.6, constMin=0.5}
  n29 = TextureCoordinate()
  n30 = Constant2Vector() {g=0.288, r=0.296}
  n31 = Panner(
   Coordinate=MaterialFunctionCall(UVs=Add(A=@n29, B=@n30), Texture Scale=Constant() {r=0.25}) {materialFunction=&ScaleUVsByCenter}.UVs
   Time=Multiply(A=Time()) {constB=10}
   Speed=Constant2Vector() {g=0.288, r=0.208}
  )
  n32 = Subtract(
   A=Saturate(None=TextureSample(UVs=@n31) {samplerType=SAMPLERTYPE_Grayscale, texture=/FortUILibrary/UIKit/Buttons/Textures/LowResBlurredNoise}.G)
   B=ScalarParameter() {defaultValue=1, parameterName=EmberReveal, sliderMax=1}
  )
  n33 = ScalarParameter() {defaultValue=0.2, parameterName=Sharpness}
  n34 = OneMinus(None=Add(A=Divide(A=@n32, B=@n33)))
  NamedRerouteDeclaration(None=Clamp(None=Multiply(A=@n28, B=@n34))) {name=Ember01 Alpha, nodeColor=<rgba 0 0 0 1>}
  n35 = Time()
  n36 = TextureSample(UVs=Panner(Coordinate=TextureCoordinate() {uTiling=1.5}, Time=Multiply(A=@n35), Speed=Constant2Vector() {g=0.264, r=-0.144})) {samplerType=SAMPLERTYPE_Grayscale, texture=/FortWater/Textures/LowResBlurredNoise}
  n37 = SmoothStep(Value=@n36.R) {constMax=0.3, constMin=0.5}
  n38 = SmoothStep(Value=@n36.R) {constMax=0.5, constMin=0.3}
  n39 = TextureCoordinate() {uTiling=2, vTiling=2}
  n40 = Multiply(A=Time()) {constB=4}
  n41 = Constant2Vector() {g=0.288, r=0.032}
  n42 = Subtract(
   A=Multiply(A=Multiply(A=@n37, B=@n38)) {constB=0.5531}
   B=TextureSample(UVs=Panner(Coordinate=@n39, Time=@n40, Speed=@n41)) {samplerType=SAMPLERTYPE_Grayscale, texture=&LowResBlurredNoise}.R
  )
  NamedRerouteDeclaration(None=Saturate(None=Multiply(A=@n42) {constB=0.6})) {name=AmbientNoise, nodeColor=<rgba 0 0 0 1>}
  n43 = Add(A=TextureCoordinate(), B=Constant2Vector())
  n44 = Constant() {r=1}
  n45 = MaterialFunctionCall(
   DistanceField=TextureSample(UVs=MaterialFunctionCall(UVs=@n43, Texture Scale=@n44) {materialFunction=&ScaleUVsByCenter}.UVs) {texture=&T_UI_DragonPop_ItemBg}.G
   Threshold=Constant() {r=0.1}
   Smoothing=Constant()
  ) {materialFunction=&MF_UI_AntiAliasing}
  n46 = NamedRerouteUsage()
  NamedRerouteDeclaration(None=Multiply(A=@n45.Outer, B=Add(A=@n45.Outer, B=@n46))) {name=MainShapeAlpha, nodeColor=<rgba 0 0 0 1>}
  NamedRerouteDeclaration(None=Max(A=NamedRerouteUsage(), B=NamedRerouteUsage())) {name=Mother of Alphas, nodeColor=<rgba 0 0 0 1>}
  NamedRerouteDeclaration(None=MaterialFunctionCall(Radius=Constant() {r=0.5}, Density=Constant() {r=5}) {materialFunction=&RadialGradientExponential}.RadialGradientExponential) {name=RadialMask, nodeColor=<rgba 0.6172 1 0 1>}
  n47 = TextureCoordinate()
  n48 = Constant2Vector() {g=-0.064, r=-0.06}
  n49 = TextureSample(UVs=MaterialFunctionCall(UVs=Add(A=@n47, B=@n48), Texture Scale=Constant() {r=1}) {materialFunction=&ScaleUVsByCenter}.UVs) {texture=&T_UI_DragonPop_ItemBg}
  n50 = Constant() {r=0.1}
  n51 = Constant() {r=0.009}
  NamedRerouteDeclaration(None=Multiply(A=MaterialFunctionCall(DistanceField=@n49.G, Threshold=@n50, Smoothing=@n51) {materialFunction=&MF_UI_AntiAliasing}.Outer) {constB=0.9}) {name=Alpha_Shadow, nodeColor=<rgba 0 0 0 1>}
  n52 = MaterialFunctionCall() {materialFunction=&LinearGradient}
  NamedRerouteDeclaration(None=@n52.UGradient) {name=U Gradient, nodeColor=<rgba 0 0 0 1>}
  NamedRerouteDeclaration(None=@n52.VGradient) {name=V Gradient, nodeColor=<rgba 0 0 0 1>}
  NamedRerouteUsage()
  NamedRerouteUsage()
  Saturate(None=MaterialFunctionCall(Speed=Constant() {r=0.4}) {materialFunction=&DebugTimeSine}.Result)
 ]
}
