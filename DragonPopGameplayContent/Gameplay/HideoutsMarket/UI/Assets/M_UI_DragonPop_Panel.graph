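# UI material for the translucent DragonPop panel in the Hideouts Market, with a call-to-action shape and currency indicator.
# Anti-aliased distance-field textures, offset and scaled, build the combined alpha from background, optional stock backing and CTA shapes.
# Color layers blend a dark base, switchable radial rarity tint and gold or diamond-blue currency gradients with screen blending.
# Scalar parameters drive a texture-based glint and a reversible black reveal mask.
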Material /DragonPopGameplayContent/Gameplay/HideoutsMarket/UI/Assets/M_UI_DragonPop_Panel
graph {
 properties {bAutomaticallySetUsageInEditor=false, blendMode=BLEND_Translucent, materialDomain=MD_UI, usedShadingModels=Default Lit}
 nodes [
  n1 = Constant3Vector() {constant=<rgba 0 0 0 1>}
  n2 = VectorParameter() {parameterName=Bg Rarity Color, defaultValue=<rgba 0.1094 0.0284 0 1>}
  n3 = TextureCoordinate()
  n4 = Constant2Vector() {g=0.012, r=0.256}
  n5 = Multiply(A=MaterialFunctionCall(UVs=Add(A=@n3, B=@n4), Radius=Constant() {r=0.3}) {materialFunction=&RadialGradientExponential}.RadialGradientExponential) {constB=0.15}
  n6 = LinearInterpolate(
   A=LinearInterpolate(B=LinearInterpolate(A=@n1, B=@n2.RGB, Alpha=@n5), Alpha=ScalarParameter() {parameterName=BgRarityColorSwitch})
   B=Constant3Vector() {constant=<rgba 0.01 0.01 0.01 1>}
   Alpha=NamedRerouteUsage()
  )
  n7 = LinearInterpolate(
   A=Constant3Vector() {constant=<rgba 1 0.7228 0.0024 1>}
   B=Constant3Vector() {constant=<rgba 0.307 0.6628 1 1>}
   Alpha=ScalarParameter() {parameterName=IsItDiamond?}
  )
  n8 = NamedRerouteUsage()
  n9 = LinearInterpolate(A=Constant3Vector() {constant=<rgba 0.1875 0.1099 0 1>}, B=Constant3Vector() {constant=<rgba 1 0.7228 0.0024 1>}, Alpha=NamedRerouteUsage())
  n10 = LinearInterpolate(A=Constant3Vector() {constant=<rgba 0.0412 0.0486 0.151 1>}, B=Constant3Vector() {constant=<rgba 0.2174 0.4312 0.7135 1>})
  n11 = ScalarParameter() {parameterName=IsItDiamond?}
  n12 = NamedRerouteUsage()
  n13 = MaterialFunctionCall!(
   Blend=Saturate(None=Multiply(A=@n7, B=@n8))
   Base=LinearInterpolate(B=LinearInterpolate(A=@n9, B=@n10, Alpha=@n11), Alpha=SmoothStep(Value=@n12) {constMax=0.4, constMin=0.2})
  ) {materialFunction=/Engine/Functions/Engine_MaterialFunctions03/Blends/Blend_Screen}
  n14 = ScalarParameter() {defaultValue=1, parameterName=CurrencyColorSwitch}
  LinearInterpolate(A=LinearInterpolate(A=@n6, B=LinearInterpolate(B=@n13.Result, Alpha=@n14), Alpha=NamedRerouteUsage()), B=@n6, Alpha=NamedRerouteUsage()) {constB=0}
  MaterialFunctionCall(Speed=Constant() {r=0.1}) {materialFunction=&DebugTimeSine}
  MaterialFunctionCall(Speed=Constant() {r=0.2}) {materialFunction=&DebugTimeSine}
  Multiply(A=NamedRerouteUsage(), B=NamedRerouteUsage())
  n15 = TextureSample() {samplerType=SAMPLERTYPE_LinearGrayscale, texture=&/T_UI_DragonPop_RevealMask}
  n16 = Step(Y=Add(A=ScalarParameter() {defaultValue=1, parameterName=BlackReveal, sliderMax=1}, B=@n15.R))
  n17 = OneMinus(None=@n16)
  Multiply(A=@n17) {constB=0.4}
  n18 = Constant2Vector() {g=1, r=0.9}
  n19 = MaterialFunctionCall(UVs=Add(A=NamedRerouteUsage(), B=Constant2Vector() {g=0.075, r=-0.06}), Texture Scale=@n18) {materialFunction=&ScaleUVsByCenter}
  n20 = Constant() {r=0.03}
  n21 = MaterialFunctionCall(
   DistanceField=TextureSample(UVs=@n19.UVs) {texture=&/T_UI_DragonPopPanel}.B
   Threshold=Constant() {r=0.5}
   Smoothing=@n20
  ) {materialFunction=&MF_UI_AntiAliasing}
  NamedRerouteDeclaration(None=@n21.Outer) {name=CTA Currency Indicator Alpha, nodeColor=<rgba 0 0 0 1>}
  n22 = TextureSample(UVs=@n19.UVs) {texture=&T_UI_DragonPopPanel}
  n23 = Frac(None=Add(A=ScalarParameter() {parameterName=Glint, sliderMax=1}, B=@n22.G))
  n24 = Multiply(
   A=Multiply(A=SmoothStep(Value=@n23) {constMax=0, constMin=0.2}, B=@n22.A)
   B=ScalarParameter() {defaultValue=1, parameterName=GlintOpacity, sliderMax=1}
  )
  NamedRerouteDeclaration(None=Multiply(A=Multiply(A=@n24) {constB=10}, B=NamedRerouteUsage())) {name=Glint, nodeColor=<rgba 0 0 0 1>}
  NamedRerouteDeclaration(None=MaterialFunctionCall() {materialFunction=&GetUserInterfaceUV}.Normalized UV) {name=Normalized UV, nodeColor=<rgba 0 0 0 1>}
  NamedRerouteDeclaration(None=MaterialFunctionCall() {materialFunction=&LinearGradient}.UGradient) {name=UGradient, nodeColor=<rgba 0 0 0 1>}
  n25 = SmoothStep(Value=NamedRerouteUsage()) {constMax=0.3, constMin=0.2}
  NamedRerouteDeclaration(None=@n25) {name=Mask, nodeColor=<rgba 0 0 0 1>}
  NamedRerouteDeclaration(None=LinearInterpolate(A=@n17, B=@n16, Alpha=ScalarParameter() {parameterName=BlackReveal Reversed})) {name=BlackRevealMask, nodeColor=<rgba 0 0 0 1>}
  n26 = MaterialFunctionCall(UVs=Add(A=NamedRerouteUsage(), B=Constant2Vector() {g=-0.214, r=0.404}), Texture Scale=Constant2Vector() {g=0.626, r=0.19}) {materialFunction=&ScaleUVsByCenter}
  n27 = Add(A=NamedRerouteUsage(), B=Constant2Vector() {r=-0.045})
  n28 = Constant2Vector() {g=0.9, r=1.05}
  n29 = MaterialFunctionCall(
   DistanceField=TextureSample(UVs=MaterialFunctionCall(UVs=@n27, Texture Scale=@n28) {materialFunction=&ScaleUVsByCenter}.UVs) {samplerType=SAMPLERTYPE_LinearGrayscale, texture=&/T_UI_DragonPopPanelBg}.R
   Threshold=Constant() {r=0.5}
  ) {materialFunction=&MF_UI_AntiAliasing}
  n30 = Max(
   A=SmoothStep(Value=TextureSample(UVs=@n26.UVs) {texture=&T_UI_DragonPopPanel}.R) {constMax=0.25, constMin=0.2}
   B=@n29.Outer
  )
  n31 = ScalarParameter() {parameterName=RemoveStockBg}
  n32 = NamedRerouteDeclaration(None=Multiply(A=LinearInterpolate(A=@n30, B=@n29.Outer, Alpha=@n31), B=ScalarParameter() {defaultValue=0.9, parameterName=PanelBgOpacity}) {constB=0.7}) {name=Panel Bg Alpha, nodeColor=<rgba 0 0 0 1>}
  n33 = Add(A=NamedRerouteUsage(), B=Constant2Vector() {g=0.05, r=-0.035})
  n34 = MaterialFunctionCall(
   DistanceField=TextureSample(UVs=MaterialFunctionCall(UVs=@n33, Texture Scale=@n18) {materialFunction=&ScaleUVsByCenter}.UVs) {texture=&T_UI_DragonPopPanel}.B
   Threshold=Constant() {r=0.5}
   Smoothing=@n20
  ) {materialFunction=&MF_UI_AntiAliasing}
  n35 = Multiply(A=Max(A=NamedRerouteDeclaration(None=@n34.Outer) {name=CTA Alpha, nodeColor=<rgba 0 0 0 1>}, B=@n21.Outer), B=@n25)
  NamedRerouteDeclaration(None=Saturate(None=Max(A=@n32, B=@n35))) {name=Combined Alpha, nodeColor=<rgba 0 0 0 1>}
  n36 = NamedRerouteUsage()
  n37 = Multiply(A=@n36, B=OneMinus(None=@n36))
  NamedRerouteDeclaration(None=Power(Base=Multiply(A=@n37) {constB=2}) {constExponent=3}) {name=Reflected UVs, nodeColor=<rgba 0 0 0 1>}
  NamedRerouteUsage()
  NamedRerouteUsage()
  ScalarParameter() {parameterName=IsItDiamond?}
 ]
}
