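# Translucent UI material for a rarity background in DragonPop's Hideouts Market interface.
# It samples a grayscale rarity background texture through 9-slice UVs from GetUserInterfaceUV.
# The red channel is slightly offset down, scaled up and saturated into a sharpened mask.
# A Rarity Color 01 vector parameter, dark purple by default, exposes the tint.
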Material /DragonPopGameplayContent/Gameplay/HideoutsMarket/UI/Assets/M_UI_DragonPop_RarityBg
graph {
 properties {bAutomaticallySetUsageInEditor=false, blendMode=BLEND_Translucent, materialDomain=MD_UI, usedShadingModels=Default Lit}
 nodes [
  n1 = TextureSample(UVs=MaterialFunctionCall() {materialFunction=&GetUserInterfaceUV}.9-Slice UV) {samplerType=SAMPLERTYPE_LinearGrayscale, texture=&/T_UI_DragonPop_RarityBg}
  Saturate(None=Multiply(A=Subtract(A=@n1.R) {constB=0.01}) {constB=5})
  VectorParameter() {parameterName=Rarity Color 01, defaultValue=<rgba 0.0096 0.0028 0.0234 1>}
 ]
}
