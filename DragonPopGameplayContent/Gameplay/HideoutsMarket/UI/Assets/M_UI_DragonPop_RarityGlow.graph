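# Additive UI material that draws an animated rarity glow for DragonPop, tinted by two Rarity Color parameters.
# Noise pans in polar coordinates at parameterized speeds, sampled with explicit derivatives from regular UVs, avoiding polar-seam mip artifacts.
# A curve-atlas falloff over an inverted circle SDF masks the noise, then a squared exponential radial gradient adds the core.
# The combined mask blends the two colors and feeds a scaled power for intensity shaping.
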Material /DragonPopGameplayContent/Gameplay/HideoutsMarket/UI/Assets/M_UI_DragonPop_RarityGlow
graph {
 properties {bAutomaticallySetUsageInEditor=false, blendMode=BLEND_Additive, materialDomain=MD_UI, usedShadingModels=Default Lit}
 nodes [
  n1 = TextureCoordinate()
  n2 = MaterialFunctionCall(Vector or UVs=Subtract(A=@n1) {constB=0.5}) {materialFunction=&VectorToRadialValue}
  n3 = ComponentMask(None=Multiply(A=@n2.Linear Distance, B=Constant() {r=1})) {r=true}
  n4 = Multiply(A=ComponentMask(None=Add(A=@n2.Radial Coordinates) {constB=0}) {r=true}, B=Constant() {r=1})
  n5 = Time()
  n6 = LinearInterpolate(Alpha=ScalarParameter() {parameterName=GlowSpeed}) {constA=1, constB=9}
  n7 = ScalarParameter() {defaultValue=0.068, parameterName=X-Speed}
  n8 = ScalarParameter() {parameterName=Y-Speed}
  n9 = Multiply(A=@n1)
  n10 = TextureSample!(
   UVs=Panner(Coordinate=AppendVector(A=@n3, B=@n4), Time=Multiply(A=@n5, B=@n6) {constB=9}, Speed=AppendVector(A=@n7, B=@n8))
   Tex
   "DDX(UVs)"=DDX(Value=@n9)
   "DDY(UVs)"=DDY(Value=@n9)
   Apply View MipBias
  ) {
   mipValueMode=TMVM_Derivative
   samplerType=SAMPLERTYPE_Grayscale
   texture=&/T_UI_DragonPop_GlowNoise
  }
  n11 = Constant()
  n12 = CurveAtlasRowParameter(CurveTime=OneMinus(None=MaterialFunctionCall(Size=@n11) {materialFunction=&MF_UI_Circle}.SDF)) {
   atlas=&/CurveAtlas_UI_DragonPop_Glow
   curve=&/Curve_UI_DragonPopGlow
  }
  n13 = Multiply(A=MaterialFunctionCall(Radius=Constant() {r=0.268}, Density=Constant() {r=0.9019}) {materialFunction=&RadialGradientExponential}.RadialGradientExponential) {constB=2}
  n14 = Add(A=Multiply(A=@n10.R, B=@n12.R), B=Multiply(A=@n13, B=@n13) {constB=2})
  LinearInterpolate(
   A=VectorParameter() {parameterName=Rarity Color 02, defaultValue=<rgba 0.4461 0 1 1>}.RGB
   B=VectorParameter() {parameterName=Rarity Color 01, defaultValue=<rgba 0.8806 0 1 1>}.RGB
   Alpha=@n14
  )
  Power(Base=Multiply(A=@n14) {constB=1.5}) {constExponent=1.5}
 ]
}
