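# Base widget for Hideouts market item details, showing the buyable item the player can interact with.
# HUD context interact events set the current item and toggle visibility through show, hide and quick-swap animation hooks, and configure the use keybind's hold and countdown display.
# Delta gates handle item and visibility changes; stock-count changes update the market item view model.
# It has no widget tree, leaving keybind, view model and animation functions as stubs.
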WidgetBlueprint /DragonPopGameplayContent/Gameplay/HideoutsMarket/UI/WBP_MarketItemDetails_Base
graphs [
 ConvertStockToText {nodes [
   n1 = ConvertStockToText!() -> (then, InStockValue:int) {is=title K2Node_FunctionEntry}
   > Return Node!(
    execute
    StockCountText:Text=Select(Option 0:Text="To Text (Integer)"(Value=@n1), Option 1:Text="LOCTEXT(\"∞\")", Index="integer < integer"(A=@n1)) -> (ReturnValue:Text)
   )
  ]}
 EventGraph {nodes [
   Event OnInitialized()
   > SetupKeybind()
   Event Construct()
   > n1 = GetFortHUDContextByPlayer!(
    execute
    LocalPlayer:Local Player Object Reference=GetOwningLocalPlayer!(self:Widget Object Reference) -> (ReturnValue:Local Player Object Reference)
   ) -> (then, ReturnValue:Fort HUDContext Object Reference)
   n2 = Create Event()
   > Bind Event to On Interact Updated(self:Fort HUDContext Object Reference=@n1, Delegate=@n2) {is=title K2Node_AddDelegate}
   > Bind Event to On Player Can Interact(self:Fort HUDContext Object Reference=@n1, Delegate=@n2) {is=title K2Node_AddDelegate}
   > SetIsShowing!(execute, self, IsShowing:bool=true) -> (then)
   > n3 = Set InteractionComponent(InteractionComponent:Fort Controller Component Interaction Object Reference=GetComponentByClass!(
     self:Actor Object Reference=GetOwningPlayer()
     ComponentClass:Actor Component Class Reference=/Script/FortniteGame.FortControllerComponent_Interaction
    ) -> (ReturnValue:Fort Controller Component Interaction Object Reference)) -> (Output_Get:Fort Controller Component Interaction Object Reference)
   > Is Valid(InputObject=@n3)
   > n4 = K2_GetInteractResponse!(execute, self:Fort Controller Component Interaction Object Reference=@n3) -> (then, ReturnValue:Fort Interact Context Info Object Reference)
   > OnPlayerCanInteract!(execute, self, ContextInfo:Fort Interact Context Info Object Reference=@n4) -> (then)
   Event Destruct()
   > n5 = GetFortHUDContextByPlayer(LocalPlayer=GetOwningLocalPlayer())
   > Is Valid(InputObject=@n5)
   > Unbind Event from On Interact Updated(self:Fort HUDContext Object Reference=@n5, Delegate=@n2) {is=title K2Node_RemoveDelegate}
   > Unbind Event from On Player Can Interact(self:Fort HUDContext Object Reference=@n5, Delegate=@n2) {is=title K2Node_RemoveDelegate}
  ]}
 GetKeybindWidget {nodes [
   GetKeybindWidget() {is=title K2Node_FunctionEntry}
   > Return Node!(execute, KeybindWidget:Keybind Widget Object Reference)
  ]}
 GetViewModel {nodes [
   GetViewModel() {is=title K2Node_FunctionEntry}
   > Return Node!(execute, ViewModel:Athena Market Item VM Object Reference)
  ]}
 OnHideAnimation {nodes [OnHideAnimation() {is=title K2Node_FunctionEntry}]}
 OnPlayerCanInteract {nodes [
   n1 = OnPlayerCanInteract!() -> (then, ContextInfo:Fort Interact Context Info Object Reference) {is=title K2Node_FunctionEntry}
   > n2 = Is Valid(InputObject=@n1)
   > n3 = Cast To BGA_Buyable_Item!(execute, Object=Get ReceivingActor!(self:Fort Interact Context Info Object Reference=@n1) -> (ReceivingActor:Actor Weak object)) -> (then, CastFailed, AsBGA Buyable Item:BGA Buyable Item Object Reference) {is=title K2Node_DynamicCast}
   > SetBuyableItemActor!(execute, self, BuyableItem:BGA Buyable Item Object Reference=@n3) -> (then)
   > n4 = Sequence()
   > Branch(Condition=Get IsShowing() -> (IsShowing:bool))
   > OnQuickSwapAnimation()
   SetIsShowing(execute=@n4.then_1, IsShowing=true)
   > UpdateInteraction!(
    execute
    self
    ContextInfo:Fort Interact Context Info Object Reference=Get ContextInfo() -> (ContextInfo:Fort Interact Context Info Object Reference)
   ) -> (then)
   SetIsShowing(execute=[@n3.CastFailed, @n2.Is Not Valid], IsShowing=false)
  ]}
 OnQuickSwapAnimation {nodes [OnQuickSwapAnimation() {is=title K2Node_FunctionEntry}]}
 OnShowAnimation {nodes [OnShowAnimation() {is=title K2Node_FunctionEntry}]}
 OnStockCountChanged {nodes [
   OnStockCountChanged() {is=title K2Node_FunctionEntry}
   n1 = GetOwningPlayer()
   > n2 = GetCurrentStock!(
    execute
    self:BGA Buyable Item Object Reference=Get Buyable Item() -> (Buyable Item:BGA Buyable Item Object Reference)
    PlayerPurchasing:Fort Player Pawn Object Reference=Cast To FortPlayerPawn!(Object=GetViewTarget!(self:Controller Object Reference=@n1) -> (ReturnValue:Actor Object Reference)) -> (AsFort Player Pawn, bSuccess:bool=~)
   ) -> (then, PersonalStockCount:int=0, GlobalStockCount:int=0)
   > UpdateMarketItemVM!(
    execute
    self:Building Gameplay Actor Market Item Object Reference=Get Buyable Item()
    ViewModel:Athena Market Item VM Object Reference=GetViewModel!(self) -> (ViewModel:Athena Market Item VM Object Reference) {is=title K2Node_CallFunction}
    PlayerController:Fort Player Controller Object Reference=Cast To FortPlayerControllerAthena!(Object=@n1) -> (AsFort Player Controller Athena, bSuccess:bool=~)
    PersonalStockCount:int=@n2
    GlobalStockCount:int=@n2.GlobalStockCount
   ) -> (then)
  ]}
 SetBuyableItemActor {nodes [
   n1 = SetBuyableItemActor!() -> (then, BuyableItem:BGA Buyable Item Object Reference) {is=title K2Node_FunctionEntry}
   > n2 = Delta Gate Wildcard!(execute:Exec, Variable:"BGA Buyable Item Object Reference (by ref)"=Get Buyable Item(), New Value:BGA Buyable Item Object Reference=@n1) -> (Success:Exec, Updated Variable:"BGA Buyable Item Object Reference (by ref)", Previous Variable:BGA Buyable Item Object Reference) {is=title K2Node_MacroInstance}
   > n3 = Is Valid(InputObject=@n2.Previous Variable)
   n4 = Create Event()
   > n5 = Unbind Event from On Stock Count Changed(self:BGA Buyable Item Object Reference=@n2.Previous Variable, Delegate=@n4) {is=title K2Node_RemoveDelegate}
   n6 = Get!(execute:Exec=[@n5.then, @n3.Is Not Valid]) -> (then:Exec, else:Exec, Buyable Item:BGA Buyable Item Object Reference)
   > Bind Event to On Stock Count Changed(self:BGA Buyable Item Object Reference=@n6, Delegate=@n4) {is=title K2Node_AddDelegate}
   > OnStockCountChanged() {is=title K2Node_CallFunction}
  ]}
 SetIsShowing {nodes [
   n1 = SetIsShowing!() -> (then, IsShowing:bool=~) {is=title K2Node_FunctionEntry}
   > n2 = Delta Gate Wildcard(Variable:"Boolean (by ref)"=Get IsShowing(), New Value:bool=@n1) -> (Updated Variable:"Boolean (by ref)", Previous Variable:bool=~)
   > n3 = Branch(Condition=@n2)
   > OnShowAnimation() {is=title K2Node_CallFunction}
   OnHideAnimation(execute=@n3.else) {is=title K2Node_CallFunction}
  ]}
 SetupKeybind {nodes [
   SetupKeybind() {is=title K2Node_FunctionEntry}
   n1 = GetKeybindWidget!(self) -> (KeybindWidget:Keybind Widget Object Reference) {is=title K2Node_CallFunction}
   > SetBoundAction!(
    execute
    self:Fort Keybind Widget Object Reference=@n1
    NewBoundAction:Name=GetUseActionName!(
     self:Fort Global UIContext Object Reference=Get FortGlobalUIContext!() -> (ReturnValue:Fort Global UIContext Object Reference) {is=title K2Node_GetBlueprintContextEx}
     OverrideInputType:ECommonInputType Enum=Get InputTypeOverride!(self:Fort Keybind Widget Object Reference=@n1) -> (InputTypeOverride:ECommonInputType Enum)
    ) -> (ReturnValue:Name=None)
   ) -> (then)
  ]}
 UpdateInteraction {nodes [
   n1 = UpdateInteraction!() -> (then, ContextInfo:Fort Interact Context Info Object Reference) {is=title K2Node_FunctionEntry}
   n2 = Get RequiredDuration!(self:Fort Interact Context Info Object Reference=@n1) -> (RequiredDuration:float)
   > SetForcedHoldKeybindStatus!(
    execute
    self:Fort Keybind Widget Object Reference=GetKeybindWidget()
    InForcedHoldKeybindStatus:EFortKeybindForcedHoldStatus Enum=Select(
     Option 0:EFortKeybindForcedHoldStatus Enum=NoForcedHold
     Option 1:EFortKeybindForcedHoldStatus Enum=ForcedHold
     Index="float > float"(A=@n2, B=0) {is=title K2Node_CallFunction}
    ) -> (ReturnValue:EFortKeybindForcedHoldStatus Enum)
   ) -> (then)
   n3 = Get ContextInfo()
   n4 = Get RequiredDuration(self=@n3)
   > SetShowProgressCountDown!(
    execute
    self:Fort Keybind Widget Object Reference=GetKeybindWidget()
    bShow:bool=AND Boolean(A="float > float"(A=@n4, B=0), B=Get bShowCountDown!(self:Fort Interact Context Info Object Reference=@n3) -> (bShowCountDown:bool))
   ) -> (then)
  ]}
]
variables [
 Buyable Item {type=BGA_Buyable_Item_C, default=null}
 InteractionComponent {type=FortControllerComponent_Interaction, default=null}
 IsShowing {type=boolean}
 Weapon ID {type=FortWeaponItemDefinition, default=/FlexLegendWeaponGameplay/Gameplay/RATT_Pistol/WID_FlexLegend_LightPistol_Athena_VR}
]
widgets {info {inheritedWidgetCount=0, parentClass=&CommonUserWidget, rootWidgetClass=null, widgetCount=0}, widgets []}
