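# Hideouts Market item-details row widget that shows a resource icon, its cost and whether the owning player can afford it.
# SetResourceAndCost takes a cost struct, sets the icon and calls CanPlayerAfford, which compares held item quantity to cost.
# If affordable, it shows the cost, or "Free" when not above zero, in white; otherwise red have/need rich text and a warning.
# It returns the affordability result to the caller.
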WidgetBlueprint /DragonPopGameplayContent/Gameplay/HideoutsMarket/UI/WBP_MarketItemDetails_ResourceAndCost
graphs [
 CanPlayerAfford {nodes [
   n1 = CanPlayerAfford!() -> (then, Resource:Item Definition Base Object Reference, Cost:int) {is=title K2Node_FunctionEntry}
   > n2 = Cast To FortPlayerController(Object=GetOwningPlayer())
   n3 = Get Item Quantity On Player!(PlayerController:Controller Object Reference=@n2, ItemDefinition:Item Definition Base Object Reference=@n1, ItemVariantGuid:Guid Structure) -> (ReturnValue:int=0) {is=title K2Node_CallFunction}
   > Return Node!(execute, CanAfford:bool="integer >= integer"(A=@n3, B=@n1.Cost), CurrentAmount:int=@n3)
  ]}
 EventGraph {links [], nodes []}
 SetResourceAndCost {nodes [
   n1 = SetResourceAndCost!() -> (then, CostData:STRUCT Hideouts Market Resource And Cost Structure) {is=title K2Node_FunctionEntry}
   n2 = Break STRUCT Hideouts Market Resource and Cost!(STRUCT_HideoutsMarket_ResourceAndCost:"STRUCT Hideouts Market Resource And Cost Structure (by ref)"=@n1) -> (ResourceType_2_#1:Item Definition Base Object Reference, Cost_5_#2:int) {is=title K2Node_BreakStruct}
   > SetBrushFromSoftTexture!(
    execute
    self:Image Object Reference=Get Image_Resource() -> (Image_Resource:Image Object Reference)
    SoftTexture:Texture 2D Soft Object Reference=GetSmallPreviewImage!(ItemDefinition:Item Definition Base Object Reference=@n2) -> (ReturnValue:Texture 2D Soft Object Reference)
    bMatchSize:bool=false
   ) -> (then)
   > n3 = CanPlayerAfford!(execute, self, Resource:Item Definition Base Object Reference=@n2, Cost:int=@n2.Cost_5_#2) -> (then, CanAfford:bool, CurrentAmount:int=0) {is=title K2Node_CallFunction}
   > n4 = Branch(Condition=@n3)
   n5 = Get Text_ResourceCount() -> (Text_ResourceCount:Common Rich Text Block Object Reference)
   > SetText(
    self=@n5
    InText=Select(
     Option 0:"Text (by ref)"=Get CostFreeText() -> (CostFreeText:Text)
     Option 1:"Text (by ref)"="To Text (Integer)"(Value=@n2.Cost_5_#2)
     Index="integer > integer"(A=@n2.Cost_5_#2)
    ) -> (ReturnValue:"Text (by ref)")
   )
   > SetDefaultColorAndOpacity!(
    execute
    self:Rich Text Block Object Reference=@n5
    InColorAndOpacity:Slate Color Structure=Make Slate Color(SpecifiedColor=(R=1,G=1,B=1,A=1))
   ) -> (then)
   > SetVisibility(self=Get Text_Warning() -> (Text_Warning:Common Text Object Reference), InVisibility=Collapsed)
   > Return Node!(execute, CanAfford=true)
   SetVisibility(execute=@n4.else, self=Get Text_Warning(), InVisibility=SelfHitTestInvisible)
   > SetText(
    self=Get Text_ResourceCount()
    InText=Format Text!(
     Format="LOCTEXT(\"{have} / <text size=\\\"20\\\">{need}</> \")"
     have:int=@n3.CurrentAmount
     need:int=@n2.Cost_5_#2
    ) -> (Result)
   )
   > SetDefaultColorAndOpacity(self=Get Text_ResourceCount(), InColorAndOpacity=Make Slate Color(SpecifiedColor=(R=1,G=0.0549,B=0.0667,A=1)))
   > Return Node(CanAfford=false)
  ]}
]
variables [CostFreeText {type=string, default=Free}]
widgets {
 info {inheritedWidgetCount=0, parentClass=&CommonUserWidget, rootWidgetClass=&HorizontalBox, widgetCount=5}
 widgets [
  HorizontalBox_25 {widgetClassPath=&HorizontalBox, widget=HorizontalBox_25}
  SizeBox_1 {
   widgetClassPath=&SizeBox
   parent=HorizontalBox_25
   widget=SizeBox_1
   properties {bOverride_HeightOverride=true, bOverride_WidthOverride=true, heightOverride=42, widthOverride=42}
   slot {class=HorizontalBoxSlot, properties {horizontalAlignment=HAlign_Center, verticalAlignment=VAlign_Center}}
  }
  Image_Resource {
   widgetClassPath=&Image
   bIsVariable=true
   parent=SizeBox_1
   widget=Image_Resource
   properties {
    brush {
     bIsDynamicallyLoaded=false
     drawAs=Image
     imageType=FullColor
     mirroring=NoMirror
     resourceName=None
     resourceObject=/Game/UI/Foundation/Textures/Icons/Athena/T-T-Icon-BR-GoldBars-UI-Icon
     tiling=NoTile
     imageSize=<128 128>
     margin=<ltrb 0 0 0 0>
     outlineSettings {
      bUseBrushTransparency=false
      roundingType=HalfHeightRadius
      width=0
      color {colorUseRule=UseColor_Specified, specifiedColor=<rgba 0 0 0 0>}
      cornerRadii=<0 0 0 0>
     }
     tintColor {colorUseRule=UseColor_Specified, specifiedColor=<rgba 1 1 1 1>}
     uVRegion {bIsValid=false, max=<0 0>, min=<0 0>}
    }
    renderTransform {angle=0, scale=<1.2 1.2>, shear=<0 0>, translation=<0 0>}
   }
   slot {class=SizeBoxSlot, properties {verticalAlignment=VAlign_Center}}
  }
  Text_ResourceCount {
   widgetClassPath=&CommonRichTextBlock
   bIsVariable=true
   parent=HorizontalBox_25
   widget=Text_ResourceCount
   properties {
    bOverrideDefaultStyle=true
    text="0 / <text size=\"20\">1</> "
    textTransformPolicy=ToUpper
    defaultTextStyleOverride {
     overflowPolicy=Clip
     transformPolicy=None
     colorAndOpacity {colorUseRule=UseColor_Specified, specifiedColor=<rgba 1 1 1 1>}
     font {
      bForceMonospaced=false
      bMaterialIsStencil=false
      fontMaterial=null
      fontObject=&Burbank
      letterSpacing=0
      monospacedWidth=1
      size=18.75
      skewAmount=0
      typefaceFontName=BurbankRegular-Black
      outlineSettings {
       bApplyOutlineToDropShadows=false
       bMiteredCorners=false
       bSeparateFillAlpha=false
       outlineMaterial=null
       outlineSize=0
       outlineColor=<rgba 0 0 0 1>
      }
     }
     highlightColor {colorUseRule=UseColor_Specified, specifiedColor=<rgba 0 0 0 1>}
     highlightShape {
      bIsDynamicallyLoaded=false
      drawAs=Image
      imageType=NoImage
      mirroring=NoMirror
      resourceName=None
      resourceObject=null
      tiling=NoTile
      imageSize=<32 32>
      margin=<ltrb 0 0 0 0>
      outlineSettings {
       bUseBrushTransparency=false
       roundingType=HalfHeightRadius
       width=0
       color {colorUseRule=UseColor_Specified, specifiedColor=<rgba 0 0 0 0>}
       cornerRadii=<0 0 0 0>
      }
      tintColor {colorUseRule=UseColor_Specified, specifiedColor=<rgba 1 1 1 1>}
      uVRegion {bIsValid=false, max=<0 0>, min=<0 0>}
     }
     selectedBackgroundColor {colorUseRule=UseColor_Foreground, specifiedColor=<rgba 1 0 1 1>}
     shadowColorAndOpacity=<rgba 0 0 0 1>
     shadowOffset=<0 0>
     strikeBrush {
      bIsDynamicallyLoaded=false
      drawAs=Image
      imageType=NoImage
      mirroring=NoMirror
      resourceName=None
      resourceObject=null
      tiling=NoTile
      imageSize=<32 32>
      margin=<ltrb 0 0 0 0>
      outlineSettings {
       bUseBrushTransparency=false
       roundingType=HalfHeightRadius
       width=0
       color {colorUseRule=UseColor_Specified, specifiedColor=<rgba 0 0 0 0>}
       cornerRadii=<0 0 0 0>
      }
      tintColor {colorUseRule=UseColor_Specified, specifiedColor=<rgba 1 1 1 1>}
      uVRegion {bIsValid=false, max=<0 0>, min=<0 0>}
     }
     underlineBrush {
      bIsDynamicallyLoaded=false
      drawAs=Image
      imageType=NoImage
      mirroring=NoMirror
      resourceName=None
      resourceObject=null
      tiling=NoTile
      imageSize=<32 32>
      margin=<ltrb 0 0 0 0>
      outlineSettings {
       bUseBrushTransparency=false
       roundingType=HalfHeightRadius
       width=0
       color {colorUseRule=UseColor_Specified, specifiedColor=<rgba 0 0 0 0>}
       cornerRadii=<0 0 0 0>
      }
      tintColor {colorUseRule=UseColor_Specified, specifiedColor=<rgba 1 1 1 1>}
      uVRegion {bIsValid=false, max=<0 0>, min=<0 0>}
     }
    }
   }
   slot {class=HorizontalBoxSlot, properties {verticalAlignment=VAlign_Center, padding=<ltrb 8 0 0 0>}}
  }
  Text_Warning {
   widgetClassPath=&CommonTextBlock
   bIsVariable=true
   parent=HorizontalBox_25
   widget=Text_Warning
   properties {
    clipping=ClipToBounds
    scrollStyle=&TextScrollStyle-Base_C
    text="Not enough!"
    textTransformPolicy=ToUpper
    colorAndOpacity {colorUseRule=UseColor_Specified, specifiedColor=<rgba 0.8469 0 0 1>}
    font {
     bForceMonospaced=false
     bMaterialIsStencil=false
     fontMaterial=null
     fontObject=&BurbankSmall
     letterSpacing=0
     monospacedWidth=1
     size=18.75
     skewAmount=0
     typefaceFontName=Black
     outlineSettings {
      bApplyOutlineToDropShadows=false
      bMiteredCorners=false
      bSeparateFillAlpha=false
      outlineMaterial=null
      outlineSize=0
      outlineColor=<rgba 0 0 0 1>
     }
    }
    shadowOffset=<0 0>
   }
   slot {
    class=HorizontalBoxSlot
    properties {horizontalAlignment=HAlign_Right, verticalAlignment=VAlign_Center, padding=<ltrb 0 0 4 0>, size {sizeRule=Fill, value=1}}
   }
  }
 ]
}
