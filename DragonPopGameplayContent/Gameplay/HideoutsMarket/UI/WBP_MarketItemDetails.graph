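# Hideouts Market item details panel widget showing an item's rarity, cost, stock and ammo icon, extending WBP_MarketItemDetails_Base.
# Setter functions push rarity colors into dynamic materials and set cost and ammo displays.
# Affordability and AthenaMarketItemVM state drive buy-button visibility, the panel's buy state (with a diamond-currency variant) and out-of-stock or limit-reached messages.
# Show, hide and quick-swap hooks play IntroV2, Outro and QuickSwap animations, which animate material parameters, opacity and transforms.
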WidgetBlueprint /DragonPopGameplayContent/Gameplay/HideoutsMarket/UI/WBP_MarketItemDetails
animations [
 IntroV2 {
  binding_count=8
  length_seconds=1.05
  object=&self:IntroV2
  sequence {
   bindings [
    Switcher_Stock {
     id=#1
     possessed_class=CommonVisibilitySwitcher
     tracks [{
       type=MovieSceneFloatTrack
       display_name=Render Opacity
       sections [{type=MovieSceneFloatSection, end_frame=17, start_frame=0, channels [None {type=MovieSceneScriptingFloatChannel, keys [[0,0],[17.866,1]]}]}]
      }]
    }
    Image_PanelBg {
     id=#2
     possessed_class=Image
     tracks [{
       type=MovieSceneWidgetMaterialTrack
       sections [{
         type=MovieSceneParameterSection
         channels [
          BlackReveal {type=MovieSceneScriptingFloatChannel, keys [[0,0],[15,1]]}
          Glint {type=MovieSceneScriptingFloatChannel, keys [[15,0.38],[38,1.227]]}
          CurrencyColorSwitch {type=MovieSceneScriptingFloatChannel, keys [[0,0],[32,1]]}
          GlintOpacity {type=MovieSceneScriptingFloatChannel, keys [[6,0],[15,1],[32,1],[34,0]]}
          BgRarityColorSwitch {type=MovieSceneScriptingFloatChannel, keys [[0,0],[6,0],[63,1]]}
          BlackReveal Reversed {type=MovieSceneScriptingFloatChannel, const [0,1]}
         ]
        }]
      }]
    }
    Overlay_RarityText {
     id=#3
     possessed_class=Overlay
     tracks [{
       type=MovieSceneFloatTrack
       display_name=Render Opacity
       sections [{type=MovieSceneFloatSection, end_frame=17, start_frame=0, channels [None {type=MovieSceneScriptingFloatChannel, keys [[0,0],[17.867,1]]}]}]
      }]
    }
    Image_BigGlow {
     id=#4
     possessed_class=Image
     tracks [{
       type=MovieSceneFloatTrack
       display_name=Render Opacity
       sections [{type=MovieSceneFloatSection, end_frame=14, start_frame=0, channels [None {type=MovieSceneScriptingFloatChannel, keys [[0,0],[14.647,0.3]]}]}]
      }]
    }
    Image_ItemBg {
     id=#5
     possessed_class=Image
     tracks [
      {
       type=MovieSceneFloatTrack
       display_name=Render Opacity
       sections [{type=MovieSceneFloatSection, channels [None {type=MovieSceneScriptingFloatChannel, default=1, const [1,1]}]}]
      }
      {
       type=MovieSceneWidgetMaterialTrack
       sections [{
         type=MovieSceneParameterSection
         channels [
          Outline Opacity {type=MovieSceneScriptingFloatChannel, keys [[4,0],[9,1],[37,1],[42,0.9]]}
          Outline Movement {type=MovieSceneScriptingFloatChannel, default=0, const [0,3]}
          InnerAlphaOpacity {type=MovieSceneScriptingFloatChannel, const [1,1]}
          OutlineScale {type=MovieSceneScriptingFloatChannel, const [0.95,3]}
          OutlineDistortion {type=MovieSceneScriptingFloatChannel, keys [[0,0.2],[15,0]]}
          BlackReveal {type=MovieSceneScriptingFloatChannel, const [1,2]}
          ColorReveal {type=MovieSceneScriptingFloatChannel, keys [[4,0],[9,1]]}
         ]
        }]
      }
     ]
    }
    Image_Item {
     id=#6
     possessed_class=FortLazyImage
     tracks [
      {
       type=MovieSceneFloatTrack
       display_name=Render Opacity
       sections [{type=MovieSceneFloatSection, channels [None {type=MovieSceneScriptingFloatChannel, keys [[0,0],[10,1]]}]}]
      }
      {
       type=MovieScene2DTransformTrack
       sections [{
         type=MovieScene2DTransformSection
         channels [
          Translation.X {type=MovieSceneScriptingFloatChannel, const [0,2]}
          Translation.Y {type=MovieSceneScriptingFloatChannel, const [0,1]}
          Angle {type=MovieSceneScriptingFloatChannel, const [0,1]}
          Scale.X {type=MovieSceneScriptingFloatChannel, default=1}
          Scale.Y {type=MovieSceneScriptingFloatChannel, default=1}
          Shear.X {type=MovieSceneScriptingFloatChannel, const [0,1]}
          Shear.Y {type=MovieSceneScriptingFloatChannel, const [0,1]}
         ]
        }]
      }
     ]
    }
    Image_ItemGlow_1 {
     id=#7
     possessed_class=Image
     tracks [
      {
       type=MovieSceneFloatTrack
       display_name=Render Opacity
       sections [{type=MovieSceneFloatSection, channels [None {type=MovieSceneScriptingFloatChannel, keys [[0,1],[13,0]]}]}]
      }
      {
       type=MovieScene2DTransformTrack
       sections [{
         type=MovieScene2DTransformSection
         end_frame=24
         start_frame=0
         channels [
          Translation.X {type=MovieSceneScriptingFloatChannel, const [0,1]}
          Translation.Y {type=MovieSceneScriptingFloatChannel, default=7, const [24,1]}
          Angle {type=MovieSceneScriptingFloatChannel, const [0,1]}
          Scale.X {type=MovieSceneScriptingFloatChannel, const [0.76,1]}
          Scale.Y {type=MovieSceneScriptingFloatChannel, keys [[0,8.9884],[5,5.3024],[13,0]]}
          Shear.X {type=MovieSceneScriptingFloatChannel, const [0,1]}
          Shear.Y {type=MovieSceneScriptingFloatChannel, const [0,1]}
         ]
        }]
      }
     ]
    }
    VerticalBox_RightSideDetails {
     id=#8
     possessed_class=VerticalBox
     tracks [
      {
       type=MovieSceneFloatTrack
       display_name=Render Opacity
       sections [{type=MovieSceneFloatSection, channels [None {type=MovieSceneScriptingFloatChannel, keys [[0,0],[10.128,1]]}]}]
      }
      {
       type=MovieScene2DTransformTrack
       sections [{
         type=MovieScene2DTransformSection
         channels [
          Translation.X {type=MovieSceneScriptingFloatChannel, keys [[0,-20],[10.134,0]]}
          Translation.Y {type=MovieSceneScriptingFloatChannel, const [0,1]}
          Angle {type=MovieSceneScriptingFloatChannel, const [0,1]}
          Scale.X {type=MovieSceneScriptingFloatChannel, const [1,1]}
          Scale.Y {type=MovieSceneScriptingFloatChannel, const [1,1]}
          Shear.X {type=MovieSceneScriptingFloatChannel, const [0,1]}
          Shear.Y {type=MovieSceneScriptingFloatChannel, const [0,1]}
         ]
        }]
      }
     ]
    }
   ]
   display_rate {numerator=60}
  }
 }
 QuickSwap {
  binding_count=9
  length_seconds=0.4
  object=&self:QuickSwap
  sequence {
   bindings [
    Switcher_Stock {
     id=#1
     possessed_class=CommonVisibilitySwitcher
     tracks [{
       type=MovieSceneFloatTrack
       display_name=Render Opacity
       sections [{type=MovieSceneFloatSection, end_frame=18, start_frame=0, channels [None {type=MovieSceneScriptingFloatChannel, const [1,1]}]}]
      }]
    }
    Image_PanelBg {
     id=#2
     possessed_class=Image
     tracks [{
       type=MovieSceneWidgetMaterialTrack
       sections [{
         type=MovieSceneParameterSection
         channels [
          BlackReveal {type=MovieSceneScriptingFloatChannel, const [1,1]}
          Glint {type=MovieSceneScriptingFloatChannel, const [1.1018,1]}
          CurrencyColorSwitch {type=MovieSceneScriptingFloatChannel, const [1,1]}
          GlintOpacity {type=MovieSceneScriptingFloatChannel, const [0,1]}
          BgRarityColorSwitch {type=MovieSceneScriptingFloatChannel, const [1,1]}
          BlackReveal Reversed {type=MovieSceneScriptingFloatChannel, const [0,1]}
         ]
        }]
      }]
    }
    Overlay_RarityText {
     id=#3
     possessed_class=Overlay
     tracks [{
       type=MovieSceneFloatTrack
       display_name=Render Opacity
       sections [{type=MovieSceneFloatSection, end_frame=18, start_frame=0, channels [None {type=MovieSceneScriptingFloatChannel, const [1,1]}]}]
      }]
    }
    Image_BigGlow {
     id=#4
     possessed_class=Image
     tracks [{
       type=MovieSceneFloatTrack
       display_name=Render Opacity
       sections [{type=MovieSceneFloatSection, end_frame=15, start_frame=0, channels [None {type=MovieSceneScriptingFloatChannel, keys [[0,1],[7,0.3]]}]}]
      }]
    }
    Image_ItemBg {
     id=#5
     possessed_class=Image
     tracks [
      {
       type=MovieSceneFloatTrack
       display_name=Render Opacity
       sections [{type=MovieSceneFloatSection, channels [None {type=MovieSceneScriptingFloatChannel, default=1, const [1,1]}]}]
      }
      {
       type=MovieSceneWidgetMaterialTrack
       sections [{
         type=MovieSceneParameterSection
         channels [
          Outline Opacity {type=MovieSceneScriptingFloatChannel, const [0.9,1]}
          Outline Movement {type=MovieSceneScriptingFloatChannel, default=0, const [0,1]}
          InnerAlphaOpacity {type=MovieSceneScriptingFloatChannel, const [1,1]}
          OutlineScale {type=MovieSceneScriptingFloatChannel, const [0.95,1]}
          OutlineDistortion {type=MovieSceneScriptingFloatChannel, keys [[0,0],[3,0.2],[8,0]]}
          BlackReveal {type=MovieSceneScriptingFloatChannel, const [1,1]}
          ColorReveal {type=MovieSceneScriptingFloatChannel, const [1,1]}
         ]
        }]
      }
     ]
    }
    Image_Item_TopGlow {
     id=#9
     possessed_class=Image
     tracks [
      {
       type=MovieSceneFloatTrack
       display_name=Render Opacity
       sections [{type=MovieSceneFloatSection, channels [None {type=MovieSceneScriptingFloatChannel, const [0,2]}]}]
      }
      {
       type=MovieScene2DTransformTrack
       sections [{
         type=MovieScene2DTransformSection
         end_frame=9
         start_frame=0
         channels [
          Translation.X {type=MovieSceneScriptingFloatChannel, const [0,1]}
          Translation.Y {type=MovieSceneScriptingFloatChannel, const [0,1]}
          Angle {type=MovieSceneScriptingFloatChannel, const [0,1]}
          Scale.X {type=MovieSceneScriptingFloatChannel, const [1.0312,1]}
          Scale.Y {type=MovieSceneScriptingFloatChannel, keys [[0,4.0264],[9,0]]}
          Shear.X {type=MovieSceneScriptingFloatChannel, const [0,1]}
          Shear.Y {type=MovieSceneScriptingFloatChannel, const [0,1]}
         ]
        }]
      }
     ]
    }
    Image_Item {
     id=#6
     possessed_class=FortLazyImage
     tracks [
      {
       type=MovieSceneFloatTrack
       display_name=Render Opacity
       sections [{type=MovieSceneFloatSection, channels [None {type=MovieSceneScriptingFloatChannel, const [1,1]}]}]
      }
      {
       type=MovieScene2DTransformTrack
       sections [{
         type=MovieScene2DTransformSection
         channels [
          Translation.X {type=MovieSceneScriptingFloatChannel, const [0,1]}
          Translation.Y {type=MovieSceneScriptingFloatChannel, const [0,1]}
          Angle {type=MovieSceneScriptingFloatChannel, default=0}
          Scale.X {type=MovieSceneScriptingFloatChannel, default=1, keys [[0,1.1],[8,1]]}
          Scale.Y {type=MovieSceneScriptingFloatChannel, default=1, keys [[0,1.1],[8,1]]}
          Shear.X {type=MovieSceneScriptingFloatChannel, const [0,1]}
          Shear.Y {type=MovieSceneScriptingFloatChannel, const [0,1]}
         ]
        }]
      }
     ]
    }
    Image_ItemGlow_1 {
     id=#7
     possessed_class=Image
     tracks [
      {
       type=MovieSceneFloatTrack
       display_name=Render Opacity
       sections [{type=MovieSceneFloatSection, channels [None {type=MovieSceneScriptingFloatChannel, keys [[0,1],[15,0]]}]}]
      }
      {
       type=MovieScene2DTransformTrack
       sections [{
         type=MovieScene2DTransformSection
         end_frame=24
         start_frame=0
         channels [
          Translation.X {type=MovieSceneScriptingFloatChannel, const [0,1]}
          Translation.Y {type=MovieSceneScriptingFloatChannel, default=7, const [18.3908,1]}
          Angle {type=MovieSceneScriptingFloatChannel, const [0,1]}
          Scale.X {type=MovieSceneScriptingFloatChannel, const [0.7484,1]}
          Scale.Y {type=MovieSceneScriptingFloatChannel, keys [[0,0],[3,6.9875],[15,0.7233]]}
          Shear.X {type=MovieSceneScriptingFloatChannel, const [0,1]}
          Shear.Y {type=MovieSceneScriptingFloatChannel, const [0,1]}
         ]
        }]
      }
     ]
    }
    VerticalBox_RightSideDetails {
     id=#8
     possessed_class=VerticalBox
     tracks [
      {
       type=MovieSceneFloatTrack
       display_name=Render Opacity
       sections [{type=MovieSceneFloatSection, channels [None {type=MovieSceneScriptingFloatChannel, const [1,1]}]}]
      }
      {
       type=MovieScene2DTransformTrack
       sections [{
         type=MovieScene2DTransformSection
         channels [
          Translation.X {type=MovieSceneScriptingFloatChannel, const [0,1]}
          Translation.Y {type=MovieSceneScriptingFloatChannel, const [0,1]}
          Angle {type=MovieSceneScriptingFloatChannel, const [0,1]}
          Scale.X {type=MovieSceneScriptingFloatChannel, const [1,1]}
          Scale.Y {type=MovieSceneScriptingFloatChannel, const [1,1]}
          Shear.X {type=MovieSceneScriptingFloatChannel, const [0,1]}
          Shear.Y {type=MovieSceneScriptingFloatChannel, const [0,1]}
         ]
        }]
      }
     ]
    }
   ]
   display_rate {numerator=60}
  }
 }
 Outro {
  binding_count=9
  length_seconds=0.6047
  object=&self:Outro
  sequence {
   bindings [
    Switcher_Stock {
     id=#1
     possessed_class=CommonVisibilitySwitcher
     tracks [{
       type=MovieSceneFloatTrack
       display_name=Render Opacity
       sections [{type=MovieSceneFloatSection, end_frame=15, start_frame=0, channels [None {type=MovieSceneScriptingFloatChannel, keys [[0,1],[10.786,0]]}]}]
      }]
    }
    Image_PanelBg {
     id=#2
     possessed_class=Image
     tracks [{
       type=MovieSceneWidgetMaterialTrack
       sections [{
         type=MovieSceneParameterSection
         channels [
          BlackReveal {type=MovieSceneScriptingFloatChannel, keys [[0,0],[23.309,1]]}
          CurrencyColorSwitch {type=MovieSceneScriptingFloatChannel, keys [[0,1],[7.03,0]]}
          BgRarityColorSwitch {type=MovieSceneScriptingFloatChannel, const [1,1]}
          BlackReveal Reversed {type=MovieSceneScriptingFloatChannel, const [1,1]}
         ]
        }]
      }]
    }
    Image_Item_TopGlow {
     id=#10
     possessed_class=Image
     tracks [{
       type=MovieSceneFloatTrack
       display_name=Render Opacity
       sections [{type=MovieSceneFloatSection, end_frame=17, start_frame=0, channels [None {type=MovieSceneScriptingFloatChannel, const [0,1]}]}]
      }]
    }
    Overlay_RarityText {
     id=#3
     possessed_class=Overlay
     tracks [{
       type=MovieSceneFloatTrack
       display_name=Render Opacity
       sections [{type=MovieSceneFloatSection, end_frame=15, start_frame=0, channels [None {type=MovieSceneScriptingFloatChannel, keys [[0,1],[10.859,0]]}]}]
      }]
    }
    Image_BigGlow {
     id=#4
     possessed_class=Image
     tracks [{
       type=MovieSceneFloatTrack
       display_name=Render Opacity
       sections [{type=MovieSceneFloatSection, channels [None {type=MovieSceneScriptingFloatChannel, keys [[0,0.3],[17.245,0]]}]}]
      }]
    }
    Image_ItemBg {
     id=#5
     possessed_class=Image
     tracks [
      {
       type=MovieSceneFloatTrack
       display_name=Render Opacity
       sections [{type=MovieSceneFloatSection, channels [None {type=MovieSceneScriptingFloatChannel, default=1, keys [[0,1],[12.649,0]]}]}]
      }
      {
       type=MovieSceneWidgetMaterialTrack
       sections [{
         type=MovieSceneParameterSection
         channels [
          Outline Opacity {type=MovieSceneScriptingFloatChannel, keys [[0,0.8],[9.155,0],[14.497,1],[30.52,1],[34.334,0.9]]}
          Outline Movement {type=MovieSceneScriptingFloatChannel, default=0, keys [[0,0],[14.497,0],[16.404,1],[34.334,0]]}
          InnerAlphaOpacity {type=MovieSceneScriptingFloatChannel, keys [[0,1],[6.486,0]]}
          OutlineScale {type=MovieSceneScriptingFloatChannel, keys [[0,0.95],[6.486,0.7],[34.334,1.3]]}
          OutlineDistortion {type=MovieSceneScriptingFloatChannel, const [0,4]}
          BlackReveal {type=MovieSceneScriptingFloatChannel, keys [[0,1],[5.722,0]]}
          ColorReveal {type=MovieSceneScriptingFloatChannel, keys [[0,1],[6.486,0]]}
         ]
        }]
      }
     ]
    }
    Image_Item {
     id=#6
     possessed_class=FortLazyImage
     tracks [
      {
       type=MovieSceneFloatTrack
       display_name=Render Opacity
       sections [{type=MovieSceneFloatSection, channels [None {type=MovieSceneScriptingFloatChannel, keys [[0,1],[16.986,0]]}]}]
      }
      {
       type=MovieScene2DTransformTrack
       sections [{
         type=MovieScene2DTransformSection
         channels [
          Translation.X {type=MovieSceneScriptingFloatChannel, keys [[0,0],[16.986,25]]}
          Translation.Y {type=MovieSceneScriptingFloatChannel, const [0,1]}
          Angle {type=MovieSceneScriptingFloatChannel, const [0,1]}
          Scale.X {type=MovieSceneScriptingFloatChannel, const [1,1]}
          Scale.Y {type=MovieSceneScriptingFloatChannel, const [1,1]}
          Shear.X {type=MovieSceneScriptingFloatChannel, const [0,1]}
          Shear.Y {type=MovieSceneScriptingFloatChannel, const [0,1]}
         ]
        }]
      }
     ]
    }
    Image_ItemGlow_1 {
     id=#7
     possessed_class=Image
     tracks [
      {
       type=MovieSceneFloatTrack
       display_name=Render Opacity
       sections [{type=MovieSceneFloatSection, channels [None {type=MovieSceneScriptingFloatChannel, const [0,1]}]}]
      }
      {
       type=MovieScene2DTransformTrack
       sections [{
         type=MovieScene2DTransformSection
         end_frame=36
         start_frame=0
         channels [
          Translation.X {type=MovieSceneScriptingFloatChannel, const [0,1]}
          Translation.Y {type=MovieSceneScriptingFloatChannel, default=7}
          Angle {type=MovieSceneScriptingFloatChannel, const [0,1]}
          Scale.X {type=MovieSceneScriptingFloatChannel, keys [[0,0.3615],[5.116,1.3],[13.954,3.0817],[18.605,0.2191]]}
          Scale.Y {type=MovieSceneScriptingFloatChannel, keys [[0,1.8931],[5.116,1.3],[13.954,1.4752],[18.605,6.8473]]}
          Shear.X {type=MovieSceneScriptingFloatChannel, const [0,1]}
          Shear.Y {type=MovieSceneScriptingFloatChannel, const [0,1]}
         ]
        }]
      }
     ]
    }
    VerticalBox_RightSideDetails {
     id=#8
     possessed_class=VerticalBox
     tracks [
      {
       type=MovieSceneFloatTrack
       display_name=Render Opacity
       sections [{type=MovieSceneFloatSection, channels [None {type=MovieSceneScriptingFloatChannel, keys [[0,1],[15.101,0]]}]}]
      }
      {
       type=MovieScene2DTransformTrack
       sections [{
         type=MovieScene2DTransformSection
         channels [
          Translation.X {type=MovieSceneScriptingFloatChannel, keys [[0,0],[9.61,20]]}
          Translation.Y {type=MovieSceneScriptingFloatChannel, const [0,1]}
          Angle {type=MovieSceneScriptingFloatChannel, const [0,1]}
          Scale.X {type=MovieSceneScriptingFloatChannel, const [1,1]}
          Scale.Y {type=MovieSceneScriptingFloatChannel, const [1,1]}
          Shear.X {type=MovieSceneScriptingFloatChannel, const [0,1]}
          Shear.Y {type=MovieSceneScriptingFloatChannel, const [0,1]}
         ]
        }]
      }
     ]
    }
   ]
   display_rate {numerator=60}
  }
 }
 Intro {
  binding_count=9
  length_seconds=1.5409
  object=&self:Intro
  sequence {
   bindings [
    Image_ItemGlow_1 {
     id=#11
     possessed_class=Image
     tracks [
      {
       type=MovieSceneFloatTrack
       display_name=Render Opacity
       sections [{type=MovieSceneFloatSection, channels [None {type=MovieSceneScriptingFloatChannel, default=1, keys [[35,0],[44,1]]}]}]
      }
      {
       type=MovieScene2DTransformTrack
       sections [{
         type=MovieScene2DTransformSection
         end_frame=65
         start_frame=35
         channels [
          Translation.X {type=MovieSceneScriptingFloatChannel, const [0,1]}
          Translation.Y {type=MovieSceneScriptingFloatChannel, const [0,1]}
          Angle {type=MovieSceneScriptingFloatChannel, const [0,1]}
          Scale.X {type=MovieSceneScriptingFloatChannel, keys [[35,0],[54,1.8132],[65,1]]}
          Scale.Y {type=MovieSceneScriptingFloatChannel, keys [[35,0],[54,1.7172],[65,1]]}
          Shear.X {type=MovieSceneScriptingFloatChannel, const [0,1]}
          Shear.Y {type=MovieSceneScriptingFloatChannel, const [0,1]}
         ]
        }]
      }
     ]
    }
    SizeBox_0 {
     id=#12
     possessed_class=SizeBox
     tracks [
      {
       type=MovieSceneFloatTrack
       display_name=Render Opacity
       sections [{type=MovieSceneFloatSection, channels [None {type=MovieSceneScriptingFloatChannel, keys [[52,0],[72.298,1]]}]}]
      }
      {
       type=MovieScene2DTransformTrack
       sections [{
         type=MovieScene2DTransformSection
         end_frame=72
         start_frame=52
         channels [
          Translation.X {type=MovieSceneScriptingFloatChannel, keys [[52,-15],[72.298,0]]}
          Translation.Y {type=MovieSceneScriptingFloatChannel, const [0,1]}
          Angle {type=MovieSceneScriptingFloatChannel, const [0,1]}
          Scale.X {type=MovieSceneScriptingFloatChannel, const [1,1]}
          Scale.Y {type=MovieSceneScriptingFloatChannel, const [1,1]}
          Shear.X {type=MovieSceneScriptingFloatChannel, const [0,1]}
          Shear.Y {type=MovieSceneScriptingFloatChannel, const [0,1]}
         ]
        }]
      }
     ]
    }
    Image_PanelBg {
     id=#13
     possessed_class=Image
     tracks [{
       type=MovieSceneWidgetMaterialTrack
       sections [{
         type=MovieSceneParameterSection
         channels [
          BorderReveal {type=MovieSceneScriptingFloatChannel, keys [[0,0],[26.384,1]]}
          BorderColorReveal {type=MovieSceneScriptingFloatChannel, keys [[7.995,0],[26.384,1]]}
          BorderOpacity {type=MovieSceneScriptingFloatChannel, keys [[74.356,1],[89.547,0.8]]}
          ColorParticleReveal {type=MovieSceneScriptingFloatChannel, keys [[32.781,0],[74.356,1]]}
          BgOpacity {type=MovieSceneScriptingFloatChannel, keys [[74.356,1],[89.547,0.9]]}
          BgColorSwap {type=MovieSceneScriptingFloatChannel, keys [[47.172,1],[89.547,0]]}
          BlackParticleReveal {type=MovieSceneScriptingFloatChannel, keys [[11.993,0],[57.566,1]]}
          BgPattern_InitialMotion {type=MovieSceneScriptingFloatChannel, keys [[11.993,0],[89.547,69.7298]]}
         ]
        }]
      }]
    }
    SizeBox_1 {
     id=#14
     possessed_class=SizeBox
     tracks [
      {
       type=MovieSceneFloatTrack
       display_name=Render Opacity
       sections [{type=MovieSceneFloatSection, channels [None {type=MovieSceneScriptingFloatChannel, keys [[69,0],[92.45,1]]}]}]
      }
      {
       type=MovieScene2DTransformTrack
       sections [{
         type=MovieScene2DTransformSection
         end_frame=92
         start_frame=69
         channels [
          Translation.X {type=MovieSceneScriptingFloatChannel, default=0}
          Translation.Y {type=MovieSceneScriptingFloatChannel, default=0}
          Angle {type=MovieSceneScriptingFloatChannel, const [0,1]}
          Scale.X {type=MovieSceneScriptingFloatChannel, const [1,1]}
          Scale.Y {type=MovieSceneScriptingFloatChannel, const [1,1]}
          Shear.X {type=MovieSceneScriptingFloatChannel, const [0,1]}
          Shear.Y {type=MovieSceneScriptingFloatChannel, const [0,1]}
         ]
        }]
      }
     ]
    }
    Overlay_RarityText {
     id=#15
     possessed_class=Overlay
     tracks [
      {
       type=MovieSceneFloatTrack
       display_name=Render Opacity
       sections [{type=MovieSceneFloatSection, channels [None {type=MovieSceneScriptingFloatChannel, keys [[42,0],[60,1]]}]}]
      }
      {
       type=MovieScene2DTransformTrack
       sections [{
         type=MovieScene2DTransformSection
         end_frame=60
         start_frame=42
         channels [
          Translation.X {type=MovieSceneScriptingFloatChannel, keys [[42,-18],[60,0]]}
          Translation.Y {type=MovieSceneScriptingFloatChannel, const [0,1]}
          Angle {type=MovieSceneScriptingFloatChannel, const [0,1]}
          Scale.X {type=MovieSceneScriptingFloatChannel, const [1,1]}
          Scale.Y {type=MovieSceneScriptingFloatChannel, const [1,1]}
          Shear.X {type=MovieSceneScriptingFloatChannel, const [0,1]}
          Shear.Y {type=MovieSceneScriptingFloatChannel, const [0,1]}
         ]
        }]
      }
     ]
    }
    Image_Item {
     id=#16
     possessed_class=FortLazyImage
     tracks [
      {
       type=MovieSceneFloatTrack
       display_name=Render Opacity
       sections [{type=MovieSceneFloatSection, channels [None {type=MovieSceneScriptingFloatChannel, keys [[27,0],[32,1]]}]}]
      }
      {
       type=MovieScene2DTransformTrack
       sections [{
         type=MovieScene2DTransformSection
         channels [
          Translation.X {type=MovieSceneScriptingFloatChannel, default=0}
          Translation.Y {type=MovieSceneScriptingFloatChannel, default=-8}
          Angle {type=MovieSceneScriptingFloatChannel, default=0}
          Scale.X {type=MovieSceneScriptingFloatChannel, keys [[23,1.5],[32,0.6],[39,1.1],[46,1]]}
          Scale.Y {type=MovieSceneScriptingFloatChannel, keys [[23,1.5],[32,0.6],[39,1.1],[46,1]]}
          Shear.X {type=MovieSceneScriptingFloatChannel, default=0}
          Shear.Y {type=MovieSceneScriptingFloatChannel, default=0}
         ]
        }]
      }
     ]
    }
    Switcher_Stock {
     id=#17
     possessed_class=CommonActivatableWidgetSwitcher
     tracks [
      {
       type=MovieSceneFloatTrack
       display_name=Render Opacity
       sections [{type=MovieSceneFloatSection, channels [None {type=MovieSceneScriptingFloatChannel, keys [[42,0],[58,1]]}]}]
      }
      {
       type=MovieScene2DTransformTrack
       sections [{
         type=MovieScene2DTransformSection
         channels [
          Translation.X {type=MovieSceneScriptingFloatChannel, keys [[42,-45],[58,0]]}
          Translation.Y {type=MovieSceneScriptingFloatChannel, const [0,1]}
          Angle {type=MovieSceneScriptingFloatChannel, const [0,1]}
          Scale.X {type=MovieSceneScriptingFloatChannel, const [1,1]}
          Scale.Y {type=MovieSceneScriptingFloatChannel, const [1,1]}
          Shear.X {type=MovieSceneScriptingFloatChannel, const [0,1]}
          Shear.Y {type=MovieSceneScriptingFloatChannel, const [0,1]}
         ]
        }]
      }
     ]
    }
    Image_Item_TopGlow {
     id=#18
     possessed_class=Image
     tracks [
      {
       type=MovieSceneFloatTrack
       display_name=Render Opacity
       sections [{type=MovieSceneFloatSection, channels [None {type=MovieSceneScriptingFloatChannel, keys [[32.999,0],[43.7,1],[74.375,0]]}]}]
      }
      {
       type=MovieScene2DTransformTrack
       sections [{
         type=MovieScene2DTransformSection
         end_frame=43
         start_frame=32
         channels [
          Translation.X {type=MovieSceneScriptingFloatChannel, const [0,1]}
          Translation.Y {type=MovieSceneScriptingFloatChannel, const [0,1]}
          Angle {type=MovieSceneScriptingFloatChannel, const [0,1]}
          Scale.X {type=MovieSceneScriptingFloatChannel, keys [[32.999,1],[38.706,4.1059],[43,1]]}
          Scale.Y {type=MovieSceneScriptingFloatChannel, keys [[32.999,1.4796],[38.706,1.2835]]}
          Shear.X {type=MovieSceneScriptingFloatChannel, const [0,1]}
          Shear.Y {type=MovieSceneScriptingFloatChannel, const [0,1]}
         ]
        }]
      }
     ]
    }
    ItemNameRow {
     id=#19
     possessed_class=HorizontalBox
     tracks [
      {
       type=MovieSceneFloatTrack
       display_name=Render Opacity
       sections [{type=MovieSceneFloatSection, channels [None {type=MovieSceneScriptingFloatChannel, keys [[52,0],[71.829,1]]}]}]
      }
      {
       type=MovieScene2DTransformTrack
       sections [{
         type=MovieScene2DTransformSection
         end_frame=71
         start_frame=52
         channels [
          Translation.X {type=MovieSceneScriptingFloatChannel, keys [[52,-15],[71.829,0]]}
          Translation.Y {type=MovieSceneScriptingFloatChannel, const [0,1]}
          Angle {type=MovieSceneScriptingFloatChannel, const [0,1]}
          Scale.X {type=MovieSceneScriptingFloatChannel, const [1,1]}
          Scale.Y {type=MovieSceneScriptingFloatChannel, const [1,1]}
          Shear.X {type=MovieSceneScriptingFloatChannel, const [0,1]}
          Shear.Y {type=MovieSceneScriptingFloatChannel, const [0,1]}
         ]
        }]
      }
     ]
    }
   ]
   display_rate {numerator=60}
  }
 }
]
graphs [
 EventGraph {nodes [
   n1 = Event PreConstruct()
   > SetBrushFromMaterial(
    self=Get Image_ItemGlow_1() -> (Image_ItemGlow_1:Image Object Reference)
    Material=/DragonPopGameplayContent/Gameplay/HideoutsMarket/UI/Assets/M_UI_DragonPop_RarityGlow
   )
   > SetBrushFromMaterial(
    self=Get Image_Item_TopGlow() -> (Image_Item_TopGlow:Image Object Reference)
    Material=&M_UI_DragonPop_RarityGlow
   )
   > SetBrushFromMaterial(
    self=Get Image_RarityTextBg() -> (Image_RarityTextBg:Image Object Reference)
    Material=/DragonPopGameplayContent/Gameplay/HideoutsMarket/UI/Assets/M_UI_DragonPop_RarityBg
   )
   > SetBrushFromMaterial(
    self=Get Image_PanelBg() -> (Image_PanelBg:Image Object Reference)
    Material=/DragonPopGameplayContent/Gameplay/HideoutsMarket/UI/Assets/M_UI_DragonPop_Panel
   )
   > Branch(Condition=@n1.IsDesignTime)
   > SetRarityData!(execute, self, Rarity:EFortRarity Enum=Common) -> (then)
  ]}
 GetKeybindWidget {nodes [
   GetKeybindWidget() {is=title K2Node_FunctionEntry}
   > Return Node!(execute, KeybindWidget:Keybind Widget Object Reference=Get KeybindWidget() -> (KeybindWidget:Keybind Widget Object Reference))
  ]}
 GetViewModel {nodes [
   GetViewModel() {is=title K2Node_FunctionEntry}
   > Return Node!(
    execute
    ViewModel:Athena Market Item VM Object Reference=Get AthenaMarketItemVM() -> (AthenaMarketItemVM:Athena Market Item VM Object Reference)
   )
  ]}
 OnHideAnimation {nodes [
   OnHideAnimation() {is=title K2Node_FunctionEntry}
   > Parent: OnHideAnimation() {is=title K2Node_CallParentFunction}
   > PlayAnimation(InAnimation=Get Outro() -> (Outro:Widget Animation Object Reference))
   > StopAnimation(InAnimation=Get IntroV2() -> (IntroV2:Widget Animation Object Reference))
  ]}
 OnQuickSwapAnimation {nodes [
   OnQuickSwapAnimation() {is=title K2Node_FunctionEntry}
   > Parent: OnQuickSwapAnimation() {is=title K2Node_CallParentFunction}
   > PlayAnimation(InAnimation=Get QuickSwap() -> (QuickSwap:Widget Animation Object Reference))
  ]}
 OnShowAnimation {nodes [
   OnShowAnimation() {is=title K2Node_FunctionEntry}
   > Parent: OnShowAnimation() {is=title K2Node_CallParentFunction}
   > PlayAnimation(InAnimation=Get IntroV2())
   > StopAnimation(InAnimation=Get Outro())
  ]}
 SetCostData {nodes [
   n1 = SetCostData!() -> (then, CostData:Athena Market Item Cost Data Structure) {is=title K2Node_FunctionEntry}
   n2 = Break Athena Market Item Cost Data!(AthenaMarketItemCostData:"Athena Market Item Cost Data Structure (by ref)"=@n1) -> (Resource:Fort Item Definition Object Reference, Cost:int, bCanPlayerAfford:bool=~) {is=title K2Node_BreakStruct}
   > Set CanAfford(CanAfford:bool=@n2.bCanPlayerAfford) -> (Output_Get:bool)
   > SetResourceAndCost!(
    execute
    self:WBP Market Item Details Resource And Cost Object Reference=Get Widget_ResourceAndCost() -> (Widget_ResourceAndCost:WBP Market Item Details Resource And Cost Object Reference)
    CostData:STRUCT Hideouts Market Resource And Cost Structure=Make STRUCT Hideouts Market Resource and Cost!(
     ResourceType_2_#20:Item Definition Base Object Reference=@n2
     Cost_5_#21:int=@n2.Cost
    ) -> (STRUCT_HideoutsMarket_ResourceAndCost:STRUCT Hideouts Market Resource And Cost Structure) {is=title K2Node_MakeStruct}
   ) -> (then, CanAfford:bool)
   n3 = "Not Equal (Integer)"(A=Get CurrentStock!(self:Athena Market Item VM Object Reference=Get AthenaMarketItemVM()) -> (CurrentStock:int=0))
   n4 = Get CanAfford() -> (CanAfford:bool)
   > SetVisibility(
    self=Get HorizontalBox_BuyButton() -> (HorizontalBox_BuyButton:Horizontal Box Object Reference)
    InVisibility=Select(Option 0=Collapsed, Option 1=SelfHitTestInvisible, Index=AND Boolean(A=@n4, B=@n3))
   )
   > n5 = GetDynamicMaterial(self=Get Image_PanelBg())
   n6 = Get CanAfford()
   n7 = Get bIsPlayerAllowedToBuy!(self:Athena Market Item VM Object Reference=Get AthenaMarketItemVM()) -> (bIsPlayerAllowedToBuy:bool)
   n8 = "Equal (Object)"(A=@n2, B=/DragonPopGameplayContent/Gameplay/DiamondCurrency/Athena_DiamondItemData)
   > SetScalarParameterValue(
    self=@n5
    ParameterName=StateSwitcher BuyPanel
    Value=Select(Option 0:float=0.5, Option 1:float=Select(Option 0=0, Option 1=1, Index=@n8), Index=AND Boolean(A=@n6, B=@n7)) -> (ReturnValue:float)
   )
  ]}
 SetItemDefinition {nodes [
   n1 = SetItemDefinition!() -> (then, ItemDefinition:Fort Item Definition Object Reference) {is=title K2Node_FunctionEntry}
   > n2 = Cast To FortWorldItemDefinition!(execute, Object=@n1) -> (then, CastFailed, AsFort World Item Definition:Fort World Item Definition Object Reference) {is=title K2Node_DynamicCast}
   > n3 = Cast To FortAmmoItemDefinition!(
    execute
    Object=GetAmmoWorldItemDefinition!(self:Fort World Item Definition Object Reference=@n2) -> (ReturnValue:Fort World Item Definition Object Reference)
   ) -> (then, CastFailed, AsFort Ammo Item Definition:Fort Ammo Item Definition Object Reference) {is=title K2Node_DynamicCast}
   > SetBrush(
    self=Get Image_AmmoIcon() -> (Image_AmmoIcon:Image Object Reference)
    InBrush=Break Fort Multi Size Brush!(FortMultiSizeBrush:"Fort Multi Size Brush Structure (by ref)"=Get AmmoIconBrush!(self:Fort Ammo Item Definition Object Reference=@n3) -> (AmmoIconBrush:Fort Multi Size Brush Structure)) -> (
     Brush_XXS:Slate Brush Structure
     Brush_XS:Slate Brush Structure
     Brush_S:Slate Brush Structure
     Brush_M:Slate Brush Structure
     Brush_L:Slate Brush Structure
     Brush_XL:Slate Brush Structure
    ) {is=title K2Node_BreakStruct}.Brush_M
   )
   n4 = Get Image_AmmoIcon()
   > SetVisibility(self=@n4, InVisibility=SelfHitTestInvisible)
   SetVisibility(execute=@n3.CastFailed, self=@n4, InVisibility=Collapsed)
  ]}
 SetRarityData {nodes [
   n1 = SetRarityData!() -> (then, Rarity:EFortRarity Enum=Common) {is=title K2Node_FunctionEntry}
   > n2 = GetRarityData!(execute, Rarity:EFortRarity Enum=@n1) -> (then, ReturnValue:Fort Rarity Item Data Structure)
   > n3 = Set RarityData(RarityData:Fort Rarity Item Data Structure=@n2) -> (Output_Get:Fort Rarity Item Data Structure)
   > n4 = GetDynamicMaterial(self=Get Image_ItemGlow_1())
   n5 = Break Fort Rarity Item Data!(FortRarityItemData:"Fort Rarity Item Data Structure (by ref)"=@n3) -> (
    Name:Text
    Color1:Linear Color Structure
    Color2:Linear Color Structure
    Color3:Linear Color Structure
    Color4:Linear Color Structure
    Color5:Linear Color Structure
    Radius:float=~
    Falloff:float=~
    Brightness:float=~
    Roughness:float=~
    Glow:float=~
   ) {is=title K2Node_BreakStruct}
   > SetVectorParameterValue(self=@n4, ParameterName=Rarity Color 01, Value=@n5.Color2)
   > SetVectorParameterValue(self=@n4, ParameterName=Rarity Color 02, Value=@n5.Color3)
   > n6 = GetDynamicMaterial(self=Get Image_Item_TopGlow())
   > SetVectorParameterValue(self=@n6, ParameterName=Rarity Color 01, Value=@n5.Color1)
   > SetVectorParameterValue(self=@n6, ParameterName=Rarity Color 02, Value=@n5.Color3)
   > "SetText (Text)"(self=Get Text_Rarity() -> (Text_Rarity:Text Block Object Reference), InText=@n5)
   > SetColorAndOpacity(self=Get Text_Rarity(), InColorAndOpacity=Make Slate Color(SpecifiedColor=@n5.Color1))
   > n7 = GetDynamicMaterial(self=Get Image_BigGlow() -> (Image_BigGlow:Image Object Reference))
   n8 = Break Fort Rarity Item Data(FortRarityItemData=Get RarityData() -> (RarityData:Fort Rarity Item Data Structure)) -> (Radius=~, Falloff=~, Brightness=~, Roughness=~, Glow=~)
   > SetVectorParameterValue(self=@n7, ParameterName=Big Glow Color 01, Value=@n8.Color1)
   > SetVectorParameterValue(self=@n7, ParameterName=Big Glow Color 02, Value=@n8.Color2)
   > n9 = GetDynamicMaterial(self=Get Image_RarityTextBg())
   > SetVectorParameterValue(
    self=@n9
    ParameterName=Rarity Color 01
    Value=Break Fort Rarity Item Data(FortRarityItemData=Get RarityData()) -> (Radius=~, Falloff=~, Brightness=~, Roughness=~, Glow=~).Color5
   )
   > n10 = GetDynamicMaterial(self=Get Image_PanelBg())
   n11 = Break Fort Rarity Item Data(FortRarityItemData=Get RarityData()) -> (Radius=~, Falloff=~, Brightness=~, Roughness=~, Glow=~)
   > SetVectorParameterValue(self=@n10, ParameterName=Rarity Color 01, Value=@n11.Color2)
   > SetVectorParameterValue(self=@n10, ParameterName=Rarity Color 02, Value=@n11.Color3)
   > SetVectorParameterValue(self=@n10, ParameterName=Bg Rarity Color, Value=@n11.Color4)
   > n12 = GetDynamicMaterial(self=Get Image_ItemBg() -> (Image_ItemBg:Image Object Reference))
   n13 = Break Fort Rarity Item Data(FortRarityItemData=Get RarityData()) -> (Radius=~, Falloff=~, Brightness=~, Roughness=~, Glow=~)
   > SetVectorParameterValue(self=@n12, ParameterName=Flame Color 01, Value=@n13.Color2)
   > SetVectorParameterValue(self=@n12, ParameterName=Flame Color 02, Value=@n13.Color3)
   > SetVectorParameterValue(self=@n12, ParameterName=Outline Color, Value=@n13.Color1)
   > SetVectorParameterValue(self=@n12, ParameterName=Inner Flame Color, Value=@n13.Color1)
  ]}
 SetStockCount {nodes [
   n1 = SetStockCount!() -> (then, Value:int) {is=title K2Node_FunctionEntry}
   > "SetText (Text)"(
    self=Get Text_StockCount() -> (Text_StockCount:Common Text Object Reference)
    InText=ConvertStockToText!(InStockValue:int=@n1) -> (StockCountText:Text)
   )
  ]}
 UpdateStockDisplay {nodes [
   UpdateStockDisplay!() -> (then, BadData:bool=~) {is=title K2Node_FunctionEntry}
   n1 = Get AthenaMarketItemVM()
   > n2 = Branch(Condition=Get bIsOutOfStock!(self:Athena Market Item VM Object Reference=@n1) -> (bIsOutOfStock:bool))
   > n3 = "SetText (Text)"(self=Get Text_StockError() -> (Text_StockError:Text Block Object Reference), InText="LOCTEXT(\"Out of\\r\\nStock!\")")
   SetActiveWidgetIndex(
    execute=[@n6.then, @n3.then]
    self:Common Visibility Switcher Object Reference=Get Switcher_Stock() -> (Switcher_Stock:Common Visibility Switcher Object Reference)
    Index=1
   )
   > n4 = GetDynamicMaterial(self=Get Image_PanelBg())
   > SetScalarParameterValue(self=@n4, ParameterName=RemoveStockBg, Value=1)
   n5 = Branch(execute=@n2.else, Condition=Get bIsLimitReached!(self:Athena Market Item VM Object Reference=@n1) -> (bIsLimitReached:bool))
   > n6 = "SetText (Text)"(self=Get Text_StockError(), InText="LOCTEXT(\"Limit\\r\\nReached!\")")
   SetActiveWidgetIndex(execute=@n5.else, self=Get Switcher_Stock(), Index=0)
   > n7 = GetDynamicMaterial(self=Get Image_PanelBg())
   > SetScalarParameterValue(self=@n7, ParameterName=RemoveStockBg, Value=0)
  ]}
]
variables [InteractionComponent_0 {type=FortControllerComponent_Interaction, default=null}]
widgets {
 info {
  inheritedWidgetCount=0
  parentClass=&/WBP_MarketItemDetails_Base.WBP_MarketItemDetails_Base_C
  rootWidgetClass=&SizeBox
  widgetCount=43
 }
 widgets [
  SizeBox_HorizontalLayout {
   widgetClassPath=&SizeBox
   widget=SizeBox_HorizontalLayout
   properties {
    bOverride_HeightOverride=true
    bOverride_WidthOverride=true
    heightOverride=185
    widthOverride=610
    renderTransform {angle=0, scale=<1 1>, shear=<0 0>, translation=<-50 0>}
   }
  }
  Overlay_228 {
   widgetClassPath=&Overlay
   parent=SizeBox_HorizontalLayout
   widget=Overlay_228
   properties {bOverride_Cursor=true}
   slot {class=SizeBoxSlot}
  }
  Image_PanelBg {
   widgetClassPath=&Image
   bIsVariable=true
   parent=Overlay_228
   widget=Image_PanelBg
   properties {
    visibility=SelfHitTestInvisible
    brush {
     bIsDynamicallyLoaded=false
     drawAs=Image
     imageType=FullColor
     mirroring=NoMirror
     resourceName=None
     resourceObject=null
     tiling=NoTile
     imageSize=<32 32>
     margin=<ltrb 0.5 0.5 0.5 0.5>
     outlineSettings {
      bUseBrushTransparency=false
      roundingType=HalfHeightRadius
      width=0
      color {colorUseRule=UseColor_Specified, specifiedColor=<rgba 0 0 0 0>}
      cornerRadii=<0 0 0 0>
     }
     tintColor {colorUseRule=UseColor_Specified, specifiedColor=<rgba 1 1 1 1>}
     uVRegion {bIsValid=false, max=<0 0>, min=<0 0>}
    }
    renderTransform {angle=0, scale=<1.25 1.2>, shear=<0 0>, translation=<-8 -6>}
   }
   slot {class=OverlaySlot, properties {horizontalAlignment=HAlign_Fill, verticalAlignment=VAlign_Fill}}
  }
  HorizontalBox_2 {
   widgetClassPath=&HorizontalBox
   parent=Overlay_228
   widget=HorizontalBox_2
   slot {class=OverlaySlot, properties {horizontalAlignment=HAlign_Fill, verticalAlignment=VAlign_Fill}}
  }
  Overlay_LeftSideItemDetails {
   widgetClassPath=&Overlay
   parent=HorizontalBox_2
   widget=Overlay_LeftSideItemDetails
   slot {class=HorizontalBoxSlot, properties {padding=<ltrb -8 0 -8 0>}}
  }
  Overlay_FeaturedItem {
   widgetClassPath=&Overlay
   parent=Overlay_LeftSideItemDetails
   widget=Overlay_FeaturedItem
   slot {class=OverlaySlot, properties {horizontalAlignment=HAlign_Fill, verticalAlignment=VAlign_Fill, padding=<ltrb 0 0 0 45>}}
  }
  Spacer_106 {widgetClassPath=&Spacer, parent=Overlay_FeaturedItem, widget=Spacer_106, properties {size=<180 0>}, slot {class=OverlaySlot}}
  Image_ItemBg {
   widgetClassPath=&Image
   bIsVariable=true
   parent=Overlay_FeaturedItem
   widget=Image_ItemBg
   properties {
    visibility=SelfHitTestInvisible
    brush {
     bIsDynamicallyLoaded=false
     drawAs=Image
     imageType=FullColor
     mirroring=NoMirror
     resourceName=None
     resourceObject=/DragonPopGameplayContent/Gameplay/HideoutsMarket/UI/Assets/M_UI_DragonPop_ItemBg
     tiling=NoTile
     imageSize=<32 32>
     margin=<ltrb 0 0 0 0>
     outlineSettings {
      bUseBrushTransparency=false
      roundingType=HalfHeightRadius
      width=0
      color {colorUseRule=UseColor_Specified, specifiedColor=<rgba 0 0 0 0>}
      cornerRadii=<0 0 0 0>
     }
     tintColor {colorUseRule=UseColor_Specified, specifiedColor=<rgba 1 1 1 1>}
     uVRegion {bIsValid=false, max=<0 0>, min=<0 0>}
    }
    renderTransform {angle=0, scale=<1.25 1.55>, shear=<0 0>, translation=<2 -2>}
   }
   slot {class=OverlaySlot, properties {horizontalAlignment=HAlign_Fill, verticalAlignment=VAlign_Fill}}
  }
  Image_ItemGlow_1 {
   widgetClassPath=&Image
   bIsVariable=true
   parent=Overlay_FeaturedItem
   widget=Image_ItemGlow_1
   properties {
    visibility=SelfHitTestInvisible
    brush {
     bIsDynamicallyLoaded=false
     drawAs=Image
     imageType=FullColor
     mirroring=NoMirror
     resourceName=None
     resourceObject=&M_UI_DragonPop_RarityGlow
     tiling=NoTile
     imageSize=<32 32>
     margin=<ltrb 0 0 0 0>
     outlineSettings {
      bUseBrushTransparency=false
      roundingType=HalfHeightRadius
      width=0
      color {colorUseRule=UseColor_Specified, specifiedColor=<rgba 0 0 0 0>}
      cornerRadii=<0 0 0 0>
     }
     tintColor {colorUseRule=UseColor_Specified, specifiedColor=<rgba 1 1 1 1>}
     uVRegion {bIsValid=false, max=<0 0>, min=<0 0>}
    }
    renderTransform {angle=0, scale=<1.9 2.4>, shear=<0 0>, translation=<0 24>}
   }
   slot {class=OverlaySlot, properties {horizontalAlignment=HAlign_Fill, verticalAlignment=VAlign_Fill}}
  }
  Image_BigGlow {
   widgetClassPath=&Image
   bIsVariable=true
   parent=Overlay_FeaturedItem
   widget=Image_BigGlow
   properties {
    renderOpacity=0.2
    visibility=SelfHitTestInvisible
    brush {
     bIsDynamicallyLoaded=false
     drawAs=Image
     imageType=FullColor
     mirroring=NoMirror
     resourceName=None
     resourceObject=/DragonPopGameplayContent/Gameplay/HideoutsMarket/UI/Assets/M_UI_DragonPop_BigGlow
     tiling=NoTile
     imageSize=<64 64>
     margin=<ltrb 0 0 0 0>
     outlineSettings {
      bUseBrushTransparency=false
      roundingType=HalfHeightRadius
      width=0
      color {colorUseRule=UseColor_Specified, specifiedColor=<rgba 0 0 0 0>}
      cornerRadii=<0 0 0 0>
     }
     tintColor {colorUseRule=UseColor_Specified, specifiedColor=<rgba 1 1 1 1>}
     uVRegion {bIsValid=false, max=<0 0>, min=<0 0>}
    }
    renderTransform {angle=0, scale=<5 5>, shear=<0 0>, translation=<0 0>}
   }
   slot {class=OverlaySlot, properties {horizontalAlignment=HAlign_Center, verticalAlignment=VAlign_Center}}
  }
  Image_Item {
   widgetClassPath=&FortLazyImage
   bIsVariable=true
   parent=Overlay_FeaturedItem
   widget=Image_Item
   properties {brush {
     bIsDynamicallyLoaded=false
     drawAs=Image
     imageType=FullColor
     mirroring=NoMirror
     resourceName=None
     resourceObject=/DragonPopGameplayContent/Gameplay/HideoutsMarket/UI/Assets/T_UI_DragonPop_TempItem
     tiling=NoTile
     imageSize=<150 150>
     margin=<ltrb 0 0 0 0>
     outlineSettings {
      bUseBrushTransparency=false
      roundingType=HalfHeightRadius
      width=0
      color {colorUseRule=UseColor_Specified, specifiedColor=<rgba 0 0 0 0>}
      cornerRadii=<0 0 0 0>
     }
     tintColor {colorUseRule=UseColor_Specified, specifiedColor=<rgba 1 1 1 1>}
     uVRegion {bIsValid=false, max=<0 0>, min=<0 0>}
    }}
   slot {class=OverlaySlot, properties {horizontalAlignment=HAlign_Center, verticalAlignment=VAlign_Center}}
  }
  Image_Item_TopGlow {
   widgetClassPath=&Image
   bIsVariable=true
   parent=Overlay_FeaturedItem
   widget=Image_Item_TopGlow
   properties {
    renderOpacity=0
    visibility=SelfHitTestInvisible
    brush {
     bIsDynamicallyLoaded=false
     drawAs=Image
     imageType=FullColor
     mirroring=NoMirror
     resourceName=None
     resourceObject=&M_UI_DragonPop_RarityGlow
     tiling=NoTile
     imageSize=<32 32>
     margin=<ltrb 0 0 0 0>
     outlineSettings {
      bUseBrushTransparency=false
      roundingType=HalfHeightRadius
      width=0
      color {colorUseRule=UseColor_Specified, specifiedColor=<rgba 0 0 0 0>}
      cornerRadii=<0 0 0 0>
     }
     tintColor {colorUseRule=UseColor_Specified, specifiedColor=<rgba 1 1 1 1>}
     uVRegion {bIsValid=false, max=<0 0>, min=<0 0>}
    }
   }
   slot {class=OverlaySlot, properties {horizontalAlignment=HAlign_Fill, verticalAlignment=VAlign_Fill}}
  }
  Overlay_RarityText {
   widgetClassPath=&Overlay
   parent=Overlay_LeftSideItemDetails
   widget=Overlay_RarityText
   slot {class=OverlaySlot, properties {horizontalAlignment=HAlign_Center, verticalAlignment=VAlign_Bottom, padding=<ltrb -10 0 0 40>}}
  }
  Spacer_RarityBgMinimum {
   widgetClassPath=&Spacer
   parent=Overlay_RarityText
   widget=Spacer_RarityBgMinimum
   properties {size=<50 1>}
   slot {class=OverlaySlot, properties {horizontalAlignment=HAlign_Center}}
  }
  Image_RarityTextBg {
   widgetClassPath=&Image
   bIsVariable=true
   parent=Overlay_RarityText
   widget=Image_RarityTextBg
   properties {
    renderOpacity=0.9
    visibility=SelfHitTestInvisible
    brush {
     bIsDynamicallyLoaded=false
     drawAs=Box
     imageType=FullColor
     mirroring=NoMirror
     resourceName=None
     resourceObject=&M_UI_DragonPop_RarityBg
     tiling=NoTile
     imageSize=<32 32>
     margin=<ltrb 0.5 0.5 0.5 0.5>
     outlineSettings {
      bUseBrushTransparency=false
      roundingType=HalfHeightRadius
      width=0
      color {colorUseRule=UseColor_Specified, specifiedColor=<rgba 0 0 0 0>}
      cornerRadii=<0 0 0 0>
     }
     tintColor {colorUseRule=UseColor_Specified, specifiedColor=<rgba 1 1 1 1>}
     uVRegion {bIsValid=false, max=<0 0>, min=<0 0>}
    }
    renderTransform {angle=0, scale=<1.2 2>, shear=<0 0>, translation=<0 -2>}
   }
   slot {class=OverlaySlot, properties {horizontalAlignment=HAlign_Fill, verticalAlignment=VAlign_Center}}
  }
  Text_Rarity {
   widgetClassPath=&UEFN_TextBlock
   bIsVariable=true
   parent=Overlay_RarityText
   widget=Text_Rarity
   properties {
    justification=Center
    text="{Legendary Long}"
    textTransformPolicy=ToUpper
    visibility=SelfHitTestInvisible
    colorAndOpacity {colorUseRule=UseColor_Specified, specifiedColor=<rgba 1 0 0.8686 1>}
    font {
     bForceMonospaced=false
     bMaterialIsStencil=false
     fontMaterial=null
     fontObject=&BurbankSmall
     letterSpacing=0
     monospacedWidth=1
     size=10.5
     skewAmount=0.24
     typefaceFontName=Black
     outlineSettings {
      bApplyOutlineToDropShadows=false
      bMiteredCorners=false
      bSeparateFillAlpha=false
      outlineMaterial=null
      outlineSize=0
      outlineColor=<rgba 0 0 0 1>
     }
    }
   }
   slot {class=OverlaySlot, properties {horizontalAlignment=HAlign_Center, verticalAlignment=VAlign_Center, padding=<ltrb 5 0 4 0>}}
  }
  Switcher_Stock {
   widgetClassPath=&CommonVisibilitySwitcher
   bIsVariable=true
   parent=Overlay_LeftSideItemDetails
   widget=Switcher_Stock
   properties {bOverride_Cursor=true}
   slot {class=OverlaySlot, properties {horizontalAlignment=HAlign_Fill, verticalAlignment=VAlign_Fill}}
  }
  CommonVisualAttachment_1 {
   widgetClassPath=&CommonVisualAttachment
   parent=Switcher_Stock
   widget=CommonVisualAttachment_1
   properties {contentAnchor=<0.5 0.5>}
   slot {class=CommonVisibilitySwitcherSlot, properties {horizontalAlignment=HAlign_Left, verticalAlignment=VAlign_Bottom}}
  }
  SizeBox_2 {
   widgetClassPath=&SizeBox
   parent=CommonVisualAttachment_1
   widget=SizeBox_2
   properties {bOverride_WidthOverride=true, widthOverride=80, renderTransform {angle=0, scale=<1 1>, shear=<0 0>, translation=<-12 -53>}}
   slot {class=SizeBoxSlot}
  }
  VerticalBox_StockText {
   widgetClassPath=&VerticalBox
   parent=SizeBox_2
   widget=VerticalBox_StockText
   properties {renderTransform {angle=0, scale=<1 1>, shear=<-11 0>, translation=<0 0>}}
   slot {class=SizeBoxSlot, properties {horizontalAlignment=HAlign_Center}}
  }
  CommonTextBlock_58 {
   widgetClassPath=&CommonTextBlock
   parent=VerticalBox_StockText
   widget=CommonTextBlock_58
   properties {
    scrollStyle=&TextScrollStyle-Base_C
    text=STOCK:
    colorAndOpacity {colorUseRule=UseColor_Specified, specifiedColor=<rgba 0.3333 0.3333 0.3333 1>}
    font {
     bForceMonospaced=false
     bMaterialIsStencil=false
     fontMaterial=null
     fontObject=&BurbankSmall
     letterSpacing=0
     monospacedWidth=1
     size=9
     skewAmount=0
     typefaceFontName=Black
     outlineSettings {
      bApplyOutlineToDropShadows=false
      bMiteredCorners=false
      bSeparateFillAlpha=false
      outlineMaterial=null
      outlineSize=2
      outlineColor=<rgba 0 0 0 0.4>
     }
    }
    shadowOffset=<0 0>
   }
   slot {class=VerticalBoxSlot, properties {horizontalAlignment=HAlign_Center, verticalAlignment=VAlign_Top, padding=<ltrb 0 20 0 5>}}
  }
  Text_StockCount {
   widgetClassPath=&CommonTextBlock
   bIsVariable=true
   parent=VerticalBox_StockText
   widget=Text_StockCount
   properties {
    text=999
    colorAndOpacity {colorUseRule=UseColor_Specified, specifiedColor=<rgba 0.3333 0.3333 0.3333 1>}
    font {
     bForceMonospaced=false
     bMaterialIsStencil=false
     fontMaterial=null
     fontObject=&BurbankSmall
     letterSpacing=0
     monospacedWidth=1
     size=10.5
     skewAmount=0
     typefaceFontName=Black
     outlineSettings {
      bApplyOutlineToDropShadows=false
      bMiteredCorners=false
      bSeparateFillAlpha=false
      outlineMaterial=null
      outlineSize=2
      outlineColor=<rgba 0 0 0 0.4>
     }
    }
    renderTransform {angle=0, scale=<1 1>, shear=<0 0>, translation=<0 -12>}
    shadowOffset=<0 0>
   }
   slot {class=VerticalBoxSlot, properties {horizontalAlignment=HAlign_Center}}
  }
  Overlay_Blurb {
   widgetClassPath=&Overlay
   parent=Switcher_Stock
   widget=Overlay_Blurb
   properties {renderTransform {angle=0, scale=<1.25 1.25>, shear=<0 0>, translation=<-88 -32>}}
   slot {class=CommonVisibilitySwitcherSlot, properties {horizontalAlignment=HAlign_Left, verticalAlignment=VAlign_Top}}
  }
  Image_Blurb {
   widgetClassPath=&Image
   bIsVariable=true
   parent=Overlay_Blurb
   widget=Image_Blurb
   properties {
    visibility=SelfHitTestInvisible
    brush {
     bIsDynamicallyLoaded=false
     drawAs=Image
     imageType=FullColor
     mirroring=NoMirror
     resourceName=None
     resourceObject=/DragonPopGameplayContent/Gameplay/HideoutsMarket/UI/Assets/M_UI_DragonPop_Blurb
     tiling=NoTile
     imageSize=<120 120>
     margin=<ltrb 0 0 0 0>
     outlineSettings {
      bUseBrushTransparency=false
      roundingType=HalfHeightRadius
      width=0
      color {colorUseRule=UseColor_Specified, specifiedColor=<rgba 0 0 0 0>}
      cornerRadii=<0 0 0 0>
     }
     tintColor {colorUseRule=UseColor_Specified, specifiedColor=<rgba 1 1 1 1>}
     uVRegion {bIsValid=false, max=<0 0>, min=<0 0>}
    }
    colorAndOpacity=<rgba 1 1 1 0.9>
    renderTransform {angle=0, scale=<1.25 1.25>, shear=<0 0>, translation=<0 0>}
   }
   slot {class=OverlaySlot, properties {horizontalAlignment=HAlign_Fill, verticalAlignment=VAlign_Fill}}
  }
  Text_StockError {
   widgetClassPath=&UEFN_TextBlock
   bIsVariable=true
   parent=Overlay_Blurb
   widget=Text_StockError
   properties {
    justification=Center
    lineHeightPercentage=0.75
    text="{error text that is too long}"
    textTransformPolicy=ToUpper
    visibility=SelfHitTestInvisible
    wrapTextAt=100
    colorAndOpacity {colorUseRule=UseColor_Specified, specifiedColor=<rgba 0.8469 0 0 1>}
    font {
     bForceMonospaced=false
     bMaterialIsStencil=false
     fontMaterial=null
     fontObject=&BurbankSmall
     letterSpacing=0
     monospacedWidth=1
     size=10.5
     skewAmount=0.24
     typefaceFontName=Black
     outlineSettings {
      bApplyOutlineToDropShadows=false
      bMiteredCorners=false
      bSeparateFillAlpha=false
      outlineMaterial=null
      outlineSize=0
      outlineColor=<rgba 0 0 0 1>
     }
    }
   }
   slot {class=OverlaySlot, properties {horizontalAlignment=HAlign_Center, verticalAlignment=VAlign_Center, padding=<ltrb 0 0 0 5>}}
  }
  VerticalBox_RightSideDetails {
   widgetClassPath=&VerticalBox
   bIsVariable=true
   parent=HorizontalBox_2
   widget=VerticalBox_RightSideDetails
   slot {class=HorizontalBoxSlot, properties {size {sizeRule=Fill, value=1}}}
  }
  ItemNameRow {
   widgetClassPath=&HorizontalBox
   parent=VerticalBox_RightSideDetails
   widget=ItemNameRow
   slot {class=VerticalBoxSlot, properties {horizontalAlignment=HAlign_Left}}
  }
  Image_AmmoIcon {
   widgetClassPath=&Image
   bIsVariable=true
   parent=ItemNameRow
   widget=Image_AmmoIcon
   properties {visibility=SelfHitTestInvisible}
   slot {class=HorizontalBoxSlot, properties {horizontalAlignment=HAlign_Center, verticalAlignment=VAlign_Center}}
  }
  Spacer_4 {widgetClassPath=&Spacer, parent=ItemNameRow, widget=Spacer_4, properties {size=<4 1>}, slot {class=HorizontalBoxSlot}}
  Text_ItemName {
   widgetClassPath=&CommonRichTextBlock
   bIsVariable=true
   parent=ItemNameRow
   widget=Text_ItemName
   properties {
    bOverrideDefaultStyle=true
    scrollStyle=&TextScrollStyle-Base_C
    text="{Item Name that is too long for this box}"
    textTransformPolicy=ToUpper
    defaultTextStyleOverride {
     overflowPolicy=Clip
     transformPolicy=None
     colorAndOpacity {colorUseRule=UseColor_Specified, specifiedColor=<rgba 1 1 1 1>}
     font {
      bForceMonospaced=false
      bMaterialIsStencil=false
      fontMaterial=null
      fontObject=&Burbank
      letterSpacing=0
      monospacedWidth=1
      size=18.75
      skewAmount=0
      typefaceFontName=BurbankRegular-Black
      outlineSettings {
       bApplyOutlineToDropShadows=false
       bMiteredCorners=false
       bSeparateFillAlpha=false
       outlineMaterial=null
       outlineSize=2
       outlineColor=<rgba 0 0 0 0.4>
      }
     }
     highlightColor {colorUseRule=UseColor_Specified, specifiedColor=<rgba 0 0 0 1>}
     highlightShape {
      bIsDynamicallyLoaded=false
      drawAs=Image
      imageType=NoImage
      mirroring=NoMirror
      resourceName=None
      resourceObject=null
      tiling=NoTile
      imageSize=<32 32>
      margin=<ltrb 0 0 0 0>
      outlineSettings {
       bUseBrushTransparency=false
       roundingType=HalfHeightRadius
       width=0
       color {colorUseRule=UseColor_Specified, specifiedColor=<rgba 0 0 0 0>}
       cornerRadii=<0 0 0 0>
      }
      tintColor {colorUseRule=UseColor_Specified, specifiedColor=<rgba 1 1 1 1>}
      uVRegion {bIsValid=false, max=<0 0>, min=<0 0>}
     }
     selectedBackgroundColor {colorUseRule=UseColor_Foreground, specifiedColor=<rgba 1 0 1 1>}
     shadowColorAndOpacity=<rgba 0 0 0 1>
     shadowOffset=<0 0>
     strikeBrush {
      bIsDynamicallyLoaded=false
      drawAs=Image
      imageType=NoImage
      mirroring=NoMirror
      resourceName=None
      resourceObject=null
      tiling=NoTile
      imageSize=<32 32>
      margin=<ltrb 0 0 0 0>
      outlineSettings {
       bUseBrushTransparency=false
       roundingType=HalfHeightRadius
       width=0
       color {colorUseRule=UseColor_Specified, specifiedColor=<rgba 0 0 0 0>}
       cornerRadii=<0 0 0 0>
      }
      tintColor {colorUseRule=UseColor_Specified, specifiedColor=<rgba 1 1 1 1>}
      uVRegion {bIsValid=false, max=<0 0>, min=<0 0>}
     }
     underlineBrush {
      bIsDynamicallyLoaded=false
      drawAs=Image
      imageType=NoImage
      mirroring=NoMirror
      resourceName=None
      resourceObject=null
      tiling=NoTile
      imageSize=<32 32>
      margin=<ltrb 0 0 0 0>
      outlineSettings {
       bUseBrushTransparency=false
       roundingType=HalfHeightRadius
       width=0
       color {colorUseRule=UseColor_Specified, specifiedColor=<rgba 0 0 0 0>}
       cornerRadii=<0 0 0 0>
      }
      tintColor {colorUseRule=UseColor_Specified, specifiedColor=<rgba 1 1 1 1>}
      uVRegion {bIsValid=false, max=<0 0>, min=<0 0>}
     }
    }
   }
   slot {class=HorizontalBoxSlot, properties {size {sizeRule=Fill, value=1}}}
  }
  Spacer_5 {widgetClassPath=&Spacer, parent=ItemNameRow, widget=Spacer_5, slot {class=HorizontalBoxSlot}}
  SizeBox_0 {
   widgetClassPath=&SizeBox
   parent=VerticalBox_RightSideDetails
   widget=SizeBox_0
   properties {bOverride_HeightOverride=true, heightOverride=71}
   slot {class=VerticalBoxSlot}
  }
  Text_Description {
   widgetClassPath=&CommonRichTextBlock
   bIsVariable=true
   parent=SizeBox_0
   widget=Text_Description
   properties {
    autoWrapText=true
    bOverrideDefaultStyle=true
    lineHeightPercentage=1.05
    scrollOrientation=Orient_Vertical
    scrollStyle=&TextScrollStyle-Base_C
    text="{Lorem ipsum dolor sit amet, consectetur adipiscing elit. Nam sagittis mauris mi, eget pharetra urna lobortis sit amet. Ut faucibus rhoncus massa, id vehicula ex tristique vel.}"
    defaultTextStyleOverride {
     overflowPolicy=Clip
     transformPolicy=None
     colorAndOpacity {colorUseRule=UseColor_Specified, specifiedColor=<rgba 0.6146 0.6146 0.6146 1>}
     font {
      bForceMonospaced=false
      bMaterialIsStencil=false
      fontMaterial=null
      fontObject=&BurbankSmall
      letterSpacing=0
      monospacedWidth=1
      size=13.5
      skewAmount=0
      typefaceFontName=Bold
      outlineSettings {
       bApplyOutlineToDropShadows=false
       bMiteredCorners=false
       bSeparateFillAlpha=false
       outlineMaterial=null
       outlineSize=0
       outlineColor=<rgba 0 0 0 0.4>
      }
     }
     highlightColor {colorUseRule=UseColor_Specified, specifiedColor=<rgba 0 0 0 1>}
     highlightShape {
      bIsDynamicallyLoaded=false
      drawAs=Image
      imageType=NoImage
      mirroring=NoMirror
      resourceName=None
      resourceObject=null
      tiling=NoTile
      imageSize=<32 32>
      margin=<ltrb 0 0 0 0>
      outlineSettings {
       bUseBrushTransparency=false
       roundingType=HalfHeightRadius
       width=0
       color {colorUseRule=UseColor_Specified, specifiedColor=<rgba 0 0 0 0>}
       cornerRadii=<0 0 0 0>
      }
      tintColor {colorUseRule=UseColor_Specified, specifiedColor=<rgba 1 1 1 1>}
      uVRegion {bIsValid=false, max=<0 0>, min=<0 0>}
     }
     selectedBackgroundColor {colorUseRule=UseColor_Foreground, specifiedColor=<rgba 1 0 1 1>}
     shadowColorAndOpacity=<rgba 0 0 0 1>
     shadowOffset=<0 0>
     strikeBrush {
      bIsDynamicallyLoaded=false
      drawAs=Image
      imageType=NoImage
      mirroring=NoMirror
      resourceName=None
      resourceObject=null
      tiling=NoTile
      imageSize=<32 32>
      margin=<ltrb 0 0 0 0>
      outlineSettings {
       bUseBrushTransparency=false
       roundingType=HalfHeightRadius
       width=0
       color {colorUseRule=UseColor_Specified, specifiedColor=<rgba 0 0 0 0>}
       cornerRadii=<0 0 0 0>
      }
      tintColor {colorUseRule=UseColor_Specified, specifiedColor=<rgba 1 1 1 1>}
      uVRegion {bIsValid=false, max=<0 0>, min=<0 0>}
     }
     underlineBrush {
      bIsDynamicallyLoaded=false
      drawAs=Image
      imageType=NoImage
      mirroring=NoMirror
      resourceName=None
      resourceObject=null
      tiling=NoTile
      imageSize=<32 32>
      margin=<ltrb 0 0 0 0>
      outlineSettings {
       bUseBrushTransparency=false
       roundingType=HalfHeightRadius
       width=0
       color {colorUseRule=UseColor_Specified, specifiedColor=<rgba 0 0 0 0>}
       cornerRadii=<0 0 0 0>
      }
      tintColor {colorUseRule=UseColor_Specified, specifiedColor=<rgba 1 1 1 1>}
      uVRegion {bIsValid=false, max=<0 0>, min=<0 0>}
     }
    }
   }
   slot {class=SizeBoxSlot}
  }
  SizeBox_1 {
   widgetClassPath=&SizeBox
   parent=VerticalBox_RightSideDetails
   widget=SizeBox_1
   properties {bOverride_HeightOverride=true, heightOverride=48, minDesiredHeight=48}
   slot {class=VerticalBoxSlot, properties {verticalAlignment=VAlign_Bottom, padding=<ltrb 0 16 0 0>}}
  }
  Switcher_BuyStates {
   widgetClassPath=&CommonActivatableWidgetSwitcher
   bIsVariable=true
   parent=SizeBox_1
   widget=Switcher_BuyStates
   slot {class=SizeBoxSlot, properties {verticalAlignment=VAlign_Center}}
  }
  UEFN_TextBlock_C_153 {
   widgetClassPath=&UEFN_TextBlock
   parent=Switcher_BuyStates
   widget=UEFN_TextBlock_C_153
   properties {
    text="You've been banned!"
    textTransformPolicy=ToUpper
    visibility=SelfHitTestInvisible
    colorAndOpacity {colorUseRule=UseColor_Specified, specifiedColor=<rgba 0.8431 0 0 1>}
    font {
     bForceMonospaced=false
     bMaterialIsStencil=false
     fontMaterial=null
     fontObject=&BurbankSmall
     letterSpacing=0
     monospacedWidth=1
     size=21
     skewAmount=0
     typefaceFontName=Black
     outlineSettings {
      bApplyOutlineToDropShadows=false
      bMiteredCorners=false
      bSeparateFillAlpha=false
      outlineMaterial=null
      outlineSize=0
      outlineColor=<rgba 0 0 0 1>
     }
    }
    renderTransform {angle=0, scale=<1 1>, shear=<0 0>, translation=<-17 0>}
   }
   slot {class=WidgetSwitcherSlot, properties {horizontalAlignment=HAlign_Right, verticalAlignment=VAlign_Center}}
  }
  HorizontalBox_InputAction {
   widgetClassPath=&HorizontalBox
   parent=Switcher_BuyStates
   widget=HorizontalBox_InputAction
   slot {class=WidgetSwitcherSlot, properties {padding=<ltrb 0 0 28 0>}}
  }
  Widget_ResourceAndCost {
   widgetClassPath=&/WBP_MarketItemDetails_ResourceAndCost
   bIsVariable=true
   parent=HorizontalBox_InputAction
   widget=Widget_ResourceAndCost
   slot {class=HorizontalBoxSlot, properties {size {sizeRule=Fill, value=1}}}
  }
  HorizontalBox_BuyButton {
   widgetClassPath=&HorizontalBox
   bIsVariable=true
   parent=HorizontalBox_InputAction
   widget=HorizontalBox_BuyButton
   slot {class=HorizontalBoxSlot, properties {horizontalAlignment=HAlign_Center, verticalAlignment=VAlign_Center}}
  }
  TouchPrompt {
   widgetClassPath=&CommonHardwareVisibilityBorder
   bIsVariable=true
   parent=HorizontalBox_BuyButton
   widget=TouchPrompt
   properties {
    padding=<ltrb 0 0 8 0>
    visibilityQuery {
     autoDescription=" ANY( Input.Touch )"
     tokenStreamVersion=0
     userDescription=isTouch
     queryTokenStream [0,1,1,1,0]
     tagDictionary [{tagName=Input.Touch}]
    }
   }
   slot {class=HorizontalBoxSlot}
  }
  Icon_TouchBuyPrompt {
   widgetClassPath=&Image
   parent=TouchPrompt
   widget=Icon_TouchBuyPrompt
   properties {
    visibility=SelfHitTestInvisible
    brush {
     bIsDynamicallyLoaded=false
     drawAs=Image
     imageType=NoImage
     mirroring=NoMirror
     resourceName=None
     resourceObject=/DragonPopGameplayContent/Gameplay/HideoutsMarket/UI/Icons/T_UI_Interact_Normal
     tiling=NoTile
     imageSize=<48 48>
     margin=<ltrb 0 0 0 0>
     outlineSettings {
      bUseBrushTransparency=false
      roundingType=HalfHeightRadius
      width=0
      color {colorUseRule=UseColor_Specified, specifiedColor=<rgba 0 0 0 0>}
      cornerRadii=<0 0 0 0>
     }
     tintColor {colorUseRule=UseColor_Specified, specifiedColor=<rgba 1 1 1 1>}
     uVRegion {bIsValid=false, max=<0 0>, min=<0 0>}
    }
   }
   slot {class=BorderSlot, properties {padding=<ltrb 0 0 8 0>}}
  }
  KeybindWidget {
   widgetClassPath=/FortUILibrary/Input/KeybindWidget
   bIsVariable=true
   parent=HorizontalBox_BuyButton
   widget=KeybindWidget
   properties {renderTransform {angle=0, scale=<0.85 0.85>, shear=<0 0>, translation=<0 0>}}
   slot {class=HorizontalBoxSlot}
  }
  BUYText_2 {
   widgetClassPath=&CommonRichTextBlock
   parent=HorizontalBox_BuyButton
   widget=BUYText_2
   properties {
    autoWrapText=true
    bOverrideDefaultStyle=true
    text=BUY
    defaultTextStyleOverride {
     overflowPolicy=Clip
     transformPolicy=None
     colorAndOpacity {colorUseRule=UseColor_Specified, specifiedColor=<rgba 1 1 1 1>}
     font {
      bForceMonospaced=false
      bMaterialIsStencil=false
      fontMaterial=null
      fontObject=&Burbank
      letterSpacing=0
      monospacedWidth=1
      size=18.75
      skewAmount=0
      typefaceFontName=BurbankRegular-Black
      outlineSettings {
       bApplyOutlineToDropShadows=false
       bMiteredCorners=false
       bSeparateFillAlpha=false
       outlineMaterial=null
       outlineSize=0
       outlineColor=<rgba 0 0 0 1>
      }
     }
     highlightColor {colorUseRule=UseColor_Specified, specifiedColor=<rgba 0 0 0 1>}
     highlightShape {
      bIsDynamicallyLoaded=false
      drawAs=Image
      imageType=NoImage
      mirroring=NoMirror
      resourceName=None
      resourceObject=null
      tiling=NoTile
      imageSize=<32 32>
      margin=<ltrb 0 0 0 0>
      outlineSettings {
       bUseBrushTransparency=false
       roundingType=HalfHeightRadius
       width=0
       color {colorUseRule=UseColor_Specified, specifiedColor=<rgba 0 0 0 0>}
       cornerRadii=<0 0 0 0>
      }
      tintColor {colorUseRule=UseColor_Specified, specifiedColor=<rgba 1 1 1 1>}
      uVRegion {bIsValid=false, max=<0 0>, min=<0 0>}
     }
     selectedBackgroundColor {colorUseRule=UseColor_Foreground, specifiedColor=<rgba 1 0 1 1>}
     shadowColorAndOpacity=<rgba 0 0 0 1>
     shadowOffset=<0 0>
     strikeBrush {
      bIsDynamicallyLoaded=false
      drawAs=Image
      imageType=NoImage
      mirroring=NoMirror
      resourceName=None
      resourceObject=null
      tiling=NoTile
      imageSize=<32 32>
      margin=<ltrb 0 0 0 0>
      outlineSettings {
       bUseBrushTransparency=false
       roundingType=HalfHeightRadius
       width=0
       color {colorUseRule=UseColor_Specified, specifiedColor=<rgba 0 0 0 0>}
       cornerRadii=<0 0 0 0>
      }
      tintColor {colorUseRule=UseColor_Specified, specifiedColor=<rgba 1 1 1 1>}
      uVRegion {bIsValid=false, max=<0 0>, min=<0 0>}
     }
     underlineBrush {
      bIsDynamicallyLoaded=false
      drawAs=Image
      imageType=NoImage
      mirroring=NoMirror
      resourceName=None
      resourceObject=null
      tiling=NoTile
      imageSize=<32 32>
      margin=<ltrb 0 0 0 0>
      outlineSettings {
       bUseBrushTransparency=false
       roundingType=HalfHeightRadius
       width=0
       color {colorUseRule=UseColor_Specified, specifiedColor=<rgba 0 0 0 0>}
       cornerRadii=<0 0 0 0>
      }
      tintColor {colorUseRule=UseColor_Specified, specifiedColor=<rgba 1 1 1 1>}
      uVRegion {bIsValid=false, max=<0 0>, min=<0 0>}
     }
    }
   }
   slot {class=HorizontalBoxSlot, properties {horizontalAlignment=HAlign_Left, verticalAlignment=VAlign_Center, size {sizeRule=Automatic, value=0.2}}}
  }
 ]
}
